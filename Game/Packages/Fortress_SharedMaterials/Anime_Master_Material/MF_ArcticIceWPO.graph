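# Material function that optionally adds wind-driven world position offset to incoming material attributes for the Arctic Ice Talus cape.
# A panned mask texture, on position-derived planar UVs, drives displacement along vertex normals blended toward a fake local-space wind direction.
# An optional gradient control blends in a weaker pure-normal offset.
# Named reroutes carry sphere and bounding-box gradient masks, and a clamped MaterialCurveA parameter scales the overall wind.
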MaterialFunction /Game/Packages/Fortress_SharedMaterials/Anime_Master_Material/MF_ArcticIceWPO
graph {nodes [
  n1 = FunctionInput() {inputType=FunctionInput_MaterialAttributes}
  n2 = GetMaterialAttributes(None=@n1) {attributeGetTypes [#1]}
  n3 = Add(A=@n2.WorldPositionOffset, B=NamedRerouteUsage())
  FunctionOutput(None=StaticSwitchParameter(
    True=SetMaterialAttributes!(MaterialAttributes=@n1, WorldPositionOffset=@n3) {attributeSetTypes [#1]}
    False=@n2.MaterialAttributes
   ) {group=ArcticIce_Cape, parameterName=UseArcticIceWPO})
  NamedRerouteDeclaration(None=SmoothStep(
    Min=ScalarParameter() {defaultValue=0.9, group=ArcticIce_Cape, parameterName=Talus_WindMask_SmoothStep_Min}
    Max=ScalarParameter() {group=ArcticIce_Cape, parameterName=Talus_WindMask_SmoothStep_Max}
    Value=MaterialFunctionCall() {materialFunction=&BoundingBoxBased_0-1_UVW}.G
   ) {constMax=0.3, constMin=0.95}) {name=Cape_WindMask_G, nodeColor=<rgba 0.3359 0 1 1>}
  NamedRerouteDeclaration(None=PreSkinnedPosition()) {name=Preskin_LocalPos, nodeColor=<rgba 0.4531 1 0 1>}
  n4 = SphereMask(
   A=NamedRerouteUsage()
   B=VectorParameter() {group=ArcticIce_Cape, parameterName=WindSphereMaskCenter, defaultValue=<rgba 0 0 20 1>}.RGB
   Radius=ScalarParameter() {defaultValue=150, group=ArcticIce_Cape, parameterName=WindSphereMaskRadius}
  ) {attenuationRadius=150, hardnessPercent=0}
  NamedRerouteDeclaration(None=Multiply(A=SmoothStep(Value=@n4), B=ScalarParameter() {defaultValue=0.3, group=ArcticIce_Cape, parameterName=WindGradient_Multi}) {constB=0.3}) {name=Cape_WindMask, nodeColor=<rgba 0.3359 0 1 1>}
  n5 = VertexNormalWS()
  n6 = Multiply(A=Normalize(VectorInput=@n5), B=ScalarParameter() {defaultValue=8, group=ArcticIce_Cape, parameterName=Wind_VertexNormal_WPO_Max})
  n7 = Divide(A=NamedRerouteUsage()) {constB=128}
  n8 = VectorParameter() {group=ArcticIce_Cape, parameterName=WindSpeed, channelNames {a="", b="", g=SpeedY, r=""}, defaultValue=<rgba 0 1.2 0 1>}
  n9 = TextureObjectParameter() {
   group=ArcticIce_Cape
   parameterName=ArcticeIceCape_TA
   samplerType=SAMPLERTYPE_Masks
   texture=/Game/Accessories/FORT_Backpacks/Backpack_M_MED_ArcticIceTalus_Cape/Textures/T_M_Med_ArcticIceTalus_Cape_TA
  }
  n10 = TextureSample(UVs=Panner(Coordinate=ComponentMask(None=@n7) {b=true, r=true}, Speed=ComponentMask(None=@n8) {g=true, r=true}) {speedY=1}, Tex=@n9) {samplerSource=SSM_Wrap_WorldGroupSettings, samplerType=SAMPLERTYPE_LinearGrayscale}
  n11 = Normalize(VectorInput=VectorParameter() {group=ArcticIce_Cape, parameterName=FakeWindDirection, defaultValue=<rgba 0 -1.5 0.5 0>}.RGB)
  n12 = Normalize(VectorInput=LinearInterpolate(
    A=@n5
    B=Transform(None=@n11) {transformSourceType=TRANSFORMSOURCE_Local}
    Alpha=ScalarParameter() {defaultValue=0.7, group=ArcticIce_Cape, parameterName=VertexNormal_WindDir_Blend}
   ) {constAlpha=1})
  n13 = Multiply(
   A=Multiply(A=@n10.R, B=NamedRerouteDeclaration(None=@n12) {name=VertexNormal, nodeColor=<rgba 0.6406 0 1 1>})
   B=ScalarParameter() {defaultValue=50, group=ArcticIce_Cape, parameterName=Wind_WPO_Max}
  ) {constB=20}
  n14 = StaticSwitchParameter(True=LinearInterpolate(A=Multiply(A=@n6, B=@n10.R), B=@n13, Alpha=NamedRerouteUsage()), False=@n13) {group=ArcticIce_Cape, parameterName=UseCapeYGradientControl}
  n15 = NamedRerouteUsage()
  n16 = ScalarParameter() {group=ArcticIce_Cape, parameterName=MaterialCurveA}
  NamedRerouteDeclaration(None=Multiply(A=Multiply(A=@n14, B=@n15), B=Clamp(None=@n16))) {name=Wind_WPO, nodeColor=<rgba 0.0547 1 0 1>}
 ]}
