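# Material function that applies an inverse film tone map to an input color, or passes it through when UseLUT is off.
# By default it samples a volume LUT with half-texel remapping for 33 or 22 resolution, then scales by a multiplier.
# UseInverseFunction instead runs analytic HLSL in ACEScg primaries on desaturated input.
# The default-on "Flip GB" switch, despite its name, feeds red into all three LUT coordinates.
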MaterialFunction /Game/Packages/Fortress_SharedMaterials/Anime_Master_Material/MF_ApplyLUT
graph {nodes [
  MaterialExpressionDesaturation_1 = Desaturation(None=NamedRerouteUsage(), Fraction=ScalarParameter() {defaultValue=0.15, group=LUT, parameterName=Desaturation, sliderMax=1})
  n1 = Custom!(ToneColor=@MaterialExpressionDesaturation_1) {
   code="// static const float3x3 XYZ_2_AP1_MAT =\r\n// {\r\n// \t 1.6410233797, -0.3248032942, -0.2364246952,\r\n// \t-0.6636628587,  1.6153315917,  0.0167563477,\r\n// \t 0.0117218943, -0.0082844420,  0.9883948585,\r\n// };\r\n\r\n// static const float3x3 D65_2_D60_CAT =\r\n// {\r\n// \t 1.0130349146, 0.0061052578, -0.0149709436,\r\n// \t 0.0076982301, 0.9981633521, -0.0050320385,\r\n// \t-0.0028413174, 0.0046851567,  0.9245061375,\r\n// };\r\n\r\n// static const float3x3 sRGB_2_XYZ_MAT =\r\n// {\r\n// \t0.4123907993, 0.3575843394, 0.1804807884,\r\n// \t0.2126390059, 0.7151686788, 0.0721923154,\r\n// \t0.0193308187, 0.1191947798, 0.9505321522,\r\n// };\r\n\r\n// static const float3x3 XYZ_2_sRGB_MAT =\r\n// {\r\n// \t 3.2409699419, -1.5373831776, -0.4986107603,\r\n// \t-0.9692436363,  1.8759675015,  0.0415550574,\r\n// \t 0.0556300797, -0.2039769589,  1.0569715142,\r\n// };\r\n\r\n// static const float3x3 D60_2_D65_CAT =\r\n// {\r\n// \t 0.9872240087, -0.0061132286, 0.0159532883,\r\n// \t-0.0075983718,  1.0018614847, 0.0053300358,\r\n// \t 0.0030725771, -0.0050959615, 1.0816806031,\r\n// };\r\n\r\n// static const float3x3 AP1_2_XYZ_MAT =\r\n// {\r\n// \t 0.6624541811, 0.1340042065, 0.1561876870,\r\n// \t 0.2722287168, 0.6740817658, 0.0536895174,\r\n// \t-0.0055746495, 0.0040607335, 1.0103391003,\r\n// };\r\n\r\nstatic const float3 AP1_RGB2Y =\r\n{\r\n\t0.2722287168, //AP1_2_XYZ_MAT[0][1],\r\n\t0.6740817658, //AP1_2_XYZ_MAT[1][1],\r\n\t0.0536895174, //AP1_2_XYZ_MAT[2][1]\r\n};\r\n\r\nstatic const float3x3 sRGB_2_AP1 =\r\n{\r\n\t0.613097402, 0.339523146, 0.047379451,\r\n\t0.070193722, 0.916353879, 0.013452398,\r\n\t0.020615592, 0.109569772, 0.869814634,\r\n};\r\n\r\nstatic const float3x3 AP1_2_sRGB =\r\n{\r\n\t 1.705050992, -0.621792120, -0.083258872,\r\n\t-0.130256417,  1.140804736, -0.010548319,\r\n\t-0.024003356, -0.128968976,  1.152972332,\r\n};\r\n\r\n//Precalculated the matrices\r\n// const float3x3 sRGB_2_AP1 = mul( XYZ_2_AP1_MAT, mul( D65_2_D60_CAT, sRGB_2_XYZ_MAT ) );\r\n// const float3x3 AP1_2_sRGB = mul( XYZ_2_sRGB_MAT, mul( D60_2_D65_CAT, AP1_2_XYZ_MAT ) );\r\n\r\n// Use ACEScg primaries as working space\r\nhalf3 WorkingColor = mul( sRGB_2_AP1, saturate( ToneColor ) );\r\nWorkingColor = max( 0, WorkingColor );\r\n\r\n// Post desaturate\r\nWorkingColor = lerp( dot( WorkingColor, AP1_RGB2Y ), WorkingColor, 1.0 / 0.93 );\r\n\r\nhalf3 ToeColor\t\t= 0.374816 * pow( 0.9 / min( WorkingColor, 0.8 ) - 1, -0.588729 );\r\nhalf3 ShoulderColor\t= 0.227986 * pow( 1.56 / ( 1.04 - WorkingColor ) - 1, 1.02046 );\r\n\r\nhalf3 t = saturate( ( WorkingColor - 0.35 ) / ( 0.45 - 0.35 ) );\r\nt = (3-2*t)*t*t;\r\nhalf3 LinearColor = lerp( ToeColor, ShoulderColor, t );\r\n\r\n// Pre desaturate\r\nLinearColor = lerp( dot( LinearColor, AP1_RGB2Y ), LinearColor, 1.0 / 0.96 );\r\nLinearColor = mul( AP1_2_sRGB, LinearColor);\r\n\r\n// Returning positive working color space values\r\nreturn max( 0, LinearColor );"
   description=FilmToneMapInverse V2
   inputs [{
     inputName=ToneColor
     input {expression=&self:MaterialExpressionDesaturation_1}
    }]
  }
  n2 = ScalarParameter() {defaultValue=1, desc=set to 1.0 for 33x33x33 resolution instead 0.0 for 22x22x22., group=LUT, parameterName=use 33 Resolution, sliderMax=1}
  n3 = FunctionInput(Preview=Constant3Vector())
  n4 = MaterialFunctionCall(Float3=@n3) {materialFunction=&BreakOutFloat3Components}
  n5 = StaticSwitchParameter(
   True=MaterialFunctionCall(X=@n4.R, Y=@n4.R, Z=@n4.R) {materialFunction=&MakeFloat3}.Result
   False=@n3
  ) {defaultValue=true, group=LUT, parameterName=Flip GB, sortPriority=0}
  n6 = Add(A=Multiply(A=@n5, B=Constant() {desc=32 / 33, r=0.9697}), B=Constant() {desc=0.5 / 33, r=0.0152}) {desc=33x33x33 resolution}
  n7 = Add(A=Multiply(A=@n5, B=Constant() {desc=21 / 22, r=0.9545}), B=Constant() {desc=0.5 / 22, r=0.0227}) {desc=22x22x22 resolution}
  n8 = TextureObjectParameter() {
   group=LUT
   parameterName=LUT Volume Texture
   sortPriority=1
   texture=&/VT_InverseLUT
  }
  n9 = Multiply(
   A=TextureSample(
    UVs=If(A=@n2, "A > B"=@n6, "A < B"=@n7) {constB=0.5, equalsThreshold=0.001}
    Tex=FunctionInput(Preview=@n8) {bUsePreviewValueAsDefault=true, inputName=LUT, inputType=FunctionInput_VolumeTexture, sortPriority=200}
   ) {texture=&VT_InverseLUT}.RGB
   B=FunctionInput(Preview=ScalarParameter() {defaultValue=3, group=LUT, parameterName=LUT Multiplier, sortPriority=5}) {bUsePreviewValueAsDefault=true, inputName=LUT Multiplier, inputType=FunctionInput_Scalar, sortPriority=300}
  )
  FunctionOutput(None=StaticSwitchParameter(True=StaticSwitchParameter(True=@n1, False=@n9) {group=LUT, parameterName=UseInverseFunction, sortPriority=0}, False=NamedRerouteUsage()) {defaultValue=true, group=LUT, parameterName=UseLUT, sortPriority=-1})
  NamedRerouteDeclaration(None=@n3) {name=Input, nodeColor=<rgba 1 0 0 1>}
 ]}
