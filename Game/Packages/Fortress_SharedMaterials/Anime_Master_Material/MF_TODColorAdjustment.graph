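# Toon material function that tints an incoming diffuse/emissive color with the time-of-day sun color at night, behind a static switch.
# Sun RGB divided by its alpha blends toward a per-material night color, then multiplies the input color, which can be desaturated.
# Two offset cosines of TimeOfDay, raised to a high power and summed, form the night gradient that, with an intensity parameter, controls the tint.
# A debug switch prints sun color and TOD values on screen.
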MaterialFunction /Game/Packages/Fortress_SharedMaterials/Anime_Master_Material/MF_TODColorAdjustment
graph {
 properties {libraryCategoriesText [Toon]}
 nodes [
  CollectionParameter() {collection=&FortniteMaterialParameters, parameterName=TimeOfDay_0-1}
  CollectionParameter() {
   collection=&FortniteMaterialParameters
   parameterName=SunLightColorWithoutLightningFlashes
  }
  CollectionParameter() {
   collection=&FortniteMaterialParameters
   parameterName=FortniteActiveTimeOfDay
  }
  CollectionParameter() {
   collection=&FortniteMaterialParameters
   parameterName=SkyLightColorAndIntensity
  }
  CollectionParameter() {
   collection=&FortniteMaterialParameters
   parameterName=CloudsDirectSunLightColor
  }
  n1 = StaticSwitchParameter(
   True=StaticSwitchParameter(True=NamedRerouteUsage(), False=NamedRerouteUsage()) {desc="RGB=Sun Color, A=0-1 TOD", group=TOD_ColorAdjustment, parameterName=ShowDebugValues, sortPriority=64}
   False=NamedRerouteUsage()
  ) {
   defaultValue=true
   desc=tints the toon diffuse color with the TOD color of the sun.
   group=TOD_ColorAdjustment
   parameterName=UseTODColorAdjustment
   sortPriority=0
  }
  FunctionOutput(None=@n1)
  n2 = CollectionParameter() {collection=&FortniteMaterialParameters, parameterName=SunLightColor}
  n3 = ComponentMask(None=@n2) {b=true, g=true, r=true}
  n4 = ComponentMask(None=@n2) {a=true}
  n5 = VectorParameter() {group=TOD_ColorAdjustment, parameterName=CustomNightColor, sortPriority=10, defaultValue=<rgba 0.099 0.223 1 1>}
  n6 = ScalarParameter() {
   desc=Blends between the Night Color of the TOD System and a per material custom color.
   group=TOD_ColorAdjustment
   parameterName=CustomNightColor_Blend
   sliderMax=1
   sortPriority=10
  }
  n7 = LinearInterpolate(A=NamedRerouteDeclaration(None=Divide(A=@n3, B=@n4)) {name=SunColor, nodeColor=<rgba 1 0.4688 0 1>}, B=@n5.RGB, Alpha=@n6)
  n8 = NamedRerouteUsage()
  n9 = ScalarParameter() {
   desc="Desaturation Intensity at night time, usualy for Yellow cosmetics."
   group=TOD_ColorAdjustment
   parameterName=NightColor_Desaturation_Intensity
   sliderMax=1
   sortPriority=20
  }
  NamedRerouteDeclaration(None=Multiply(A=Saturate(None=@n7), B=Desaturation(None=@n8, Fraction=@n9))) {name=CombinedColor, nodeColor=<rgba 0.0312 1 0 1>}
  n10 = Multiply(A=ScreenPosition().ViewportUV, B=Constant2Vector() {g=3, r=8}) {constB=12}
  n11 = MaterialFunctionCall!(
   Vector4=AppendVector(
    A=NamedRerouteUsage()
    B=CollectionParameter() {collection=&FortniteMaterialParameters, parameterName=TimeOfDay_0-1}
   )
   MaximumNumberOfDigits=Constant() {r=4}
   UVs=Frac(None=Add(A=@n10) {constB=0.5})
   DebugTextLocation RG_UpperRight BA_LowerLeft=Constant4Vector() {constant=<rgba 0 0.25 1 0.75>}.RGBA
   Component Spacing=Constant4Vector() {constant=<rgba 0.2 0 1 0.65>}.RGBA
  ) {materialFunction=/Engine/Functions/Engine_MaterialFunctions02/Utility/DebugFloat4Values}
  NamedRerouteDeclaration(None=LinearInterpolate(A=NamedRerouteUsage(), B=ComponentMask(None=@n11.ColorCodedOutput) {b=true, g=true, r=true}, Alpha=@n11.GreyScaleOutput) {constA=1, constB=0}) {name=DebugColor, nodeColor=<rgba 1 0.1972 0 1>}
  n12 = CollectionParameter() {collection=&FortniteMaterialParameters, parameterName=TimeOfDay_0-1}
  n13 = Saturate(None=Add(A=@n12) {constB=-0.15})
  n14 = Add(A=Multiply(A=Cosine(None=@n13)) {constB=0.5}) {constB=0.5}
  n15 = Saturate(None=Add(A=@n12) {constB=0.15})
  n16 = Add(A=Multiply(A=Cosine(None=@n15)) {constB=0.5}) {constB=0.5}
  NamedRerouteDeclaration(None=Add(A=Power(Base=@n14) {constExponent=32}, B=Power(Base=@n16) {constExponent=32})) {name=NightTimeGradient, nodeColor=<rgba 0.1719 0 1 1>}
  NamedRerouteDeclaration(None=FunctionInput() {description="Toon Emissive Color (Diffuse) in", inputName=Color}) {name=DiffuseColor_In, nodeColor=<rgba 0 1 0.9766 1>}
  n17 = NamedRerouteUsage()
  n18 = NamedRerouteUsage()
  n19 = NamedRerouteUsage()
  n20 = ScalarParameter() {
   defaultValue=0.7
   desc="1 = full night color tinting, 0 = no night color tinting (full texture color no darkening)"
   group=TOD_ColorAdjustment
   parameterName=NightColor_Instenity
   sliderMax=1
   sortPriority=5
  }
  NamedRerouteDeclaration(None=LinearInterpolate(A=@n17, B=LinearInterpolate(A=@n17, B=@n18, Alpha=@n19), Alpha=@n20)) {name=DiffuseColor_Out, nodeColor=<rgba 1 0.9831 0.0514 1>}
 ]
}
