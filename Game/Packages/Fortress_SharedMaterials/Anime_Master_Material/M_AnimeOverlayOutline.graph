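# Master unlit, masked anime outline overlay material for skinned characters, drawing an inverted hull by pushing vertices along normals and showing only backfaces.
# Thickness optionally scales with camera distance, capped by FOV, and a SkinFX mask; normal-map directions avoid segmented geometry.
# Vertex colors can collapse hidden vertices, and textures can mask lines.
# Flat, textured or customizable color gets LUT and time-of-day adjustment before cloth-sim, de-rez and eye-adaptation functions.
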Material /Game/Packages/Fortress_SharedMaterials/Anime_Master_Material/M_AnimeOverlayOutline
graph {
 properties {
  bAutomaticallySetUsageInEditor=false
  bEnableResponsiveAA=true
  bUseMaterialAttributes=true
  bUsedWithClothing=true
  bUsedWithMeshDeformer=true
  bUsedWithMorphTargets=true
  bUsedWithSkeletalMesh=true
  blendMode=BLEND_Masked
  pixelDepthOffsetMode=PDOM_Legacy
  previewMesh=/BRCosmetics/Characters/Player/Female/Medium/Bodies/F_MED_ConfectionPop/Meshes/F_MED_ConfectionPop
  shadingModel=MSM_Unlit
  twoSided=true
  usedShadingModels=Unlit
 }
 nodes [
  n1 = VertexColor()
  n2 = ChannelMaskParameter(None=AppendVector(A=@n1, B=@n1)) {group=OverlayOutlines, maskChannel=Green, parameterName=VertexHide Channel, sortPriority=2, defaultValue=<rgba 0 1 0 0>}
  n3 = ChannelMaskParameter(None=@n1) {group=OverlayOutlines, maskChannel=Green, parameterName=VertexHideMaskChannel, sortPriority=2, defaultValue=<rgba 0 1 0 0>}
  n4 = LinearInterpolate(
   A=SphereMask(
    A=@n2.R
    B=ScalarParameter() {defaultValue=0.1, group=OverlayOutlines, parameterName=VertexHide Threshold, sliderMax=1}
    Radius=Constant() {r=0.13}
    Hardness=Constant() {r=1}
   ) {attenuationRadius=0.1}
   Alpha=@n3.R
  )
  n5 = NamedRerouteUsage()
  n6 = ScalarParameter() {defaultValue=0.05, group=OverlayOutlines, parameterName=LineThickness, sliderMax=1, sortPriority=3}
  n7 = StaticSwitchParameter(
   True=LinearInterpolate(
    A=@n6
    B=ScalarParameter() {defaultValue=0.6, group=OverlayOutlines, parameterName=LineThicknessFar, sliderMax=4, sortPriority=3}
    Alpha=NamedRerouteUsage()
   )
   False=@n6
  ) {group=OverlayOutlines, parameterName=UseDistanceFading, sortPriority=0}
  n8 = TextureSampleParameter2D() {group=OverlayOutlines, parameterName=NormalMap, samplerType=SAMPLERTYPE_Normal, texture=&FlatNormalBlue}
  n9 = Multiply(
   A=StaticSwitchParameter(True=Multiply(A=@n5, B=@n7), False=@n7) {group=OverlayOutlines, parameterName=UseSkinFX_Thickness, sortPriority=0}
   B=StaticSwitchParameter(True=Transform(None=@n8.RGB), False=VertexNormalWS()) {desc=use normal map to prevent segmented geometry, group=OverlayOutlines, parameterName=UseNormalMap, sortPriority=0}
  )
  If(A=@n4, B=Constant() {r=0.5}, "A > B"=Divide(A=Constant() {r=1}, B=Constant()) {constB=0, desc=NaN clipping}, "A < B"=@n9)
  n10 = VectorParameter() {group=OverlayOutlines, parameterName=Color, sortPriority=2, defaultValue=<rgba 0.005 0.005 0.005 1>}
  n11 = VectorParameter() {group=CustomizableColor, parameterName=CustomColor_Default, sortPriority=2, defaultValue=<rgba 0 0 0 1>}
  n12 = VectorParameter() {group=CustomizableColor, parameterName=CustomColor, sortPriority=2, defaultValue=<rgba 1 1 1 1>}
  n13 = TextureSampleParameter2D() {
   group=CustomizableColor
   parameterName=ColorCustomizationMap
   samplerType=SAMPLERTYPE_Masks
   sortPriority=1
   texture=&T_BlackMask
  }
  n14 = TextureSampleParameter2D() {
   desc=Texture for Outline Diffuse Color
   group=OverlayOutlines
   parameterName=Diffuse
   sortPriority=1
   texture=&T_BlacksRGB
  }
  n15 = ScalarParameter() {
   defaultValue=1
   desc=Multiplier for the Outline Diffuse Texture
   group=OverlayOutlines
   parameterName=DiffuseMult
   sliderMax=1
   sortPriority=3
  }
  n16 = StaticSwitchParameter(True=Multiply(A=@n14.RGB, B=@n15), False=@n10.RGB) {desc=Enable to use a texture to color the edge outlines of the mesh., group=OverlayOutlines, parameterName=UseDiffuseMap, sortPriority=0}
  n17 = StaticSwitchParameter(
   True=LinearInterpolate(
    A=@n10.RGB
    B=LinearInterpolate(A=@n11.RGB, B=@n12.RGB, Alpha=@n12.RGB)
    Alpha=ChannelMaskParameter(None=@n13.RGB) {group=CustomizableColor, parameterName=CustomColorMaskChannel, sortPriority=3}.R
   )
   False=@n16
  ) {
   desc="Switch to Activate Color Customization.\r\nOutlineColor is the default color, and CustomColor the color controlled by the mask.\r\nDiffuseMult_Outline is used to darken the tint color "
   group=CustomizableColor
   parameterName=UseCustomColor
   sortPriority=0
  }
  n18 = SetMaterialAttributes!(
   MaterialAttributes
   Emissive Color=MaterialFunctionCall!(Color=MaterialFunctionCall!(In=@n17, LUT, LUT Multiplier) {materialFunction=&/MF_ApplyLUT}.Result) {materialFunction=&/MF_TODColorAdjustment}.Result
   Opacity Mask=NamedRerouteUsage()
   World Position Offset=NamedRerouteUsage()
  ) {attributeSetTypes [#1, #2, #3]}
  MaterialFunctionCall!(
   Emissive
   Alpha=ScalarParameter() {defaultValue=1, group=Cosmetic Emissive EyeAdaptation, parameterName=Cosmetic Emissive EyeAdaptation Alpha}
   Material=MaterialFunctionCall!(
    Fresnel
    Material=MaterialFunctionCall!(Material=@n18, SkinFX_Mask) {materialFunction=/Game/Global/MaterialFunctions/MF_ClothSimWPO_v2}.Result
   ) {materialFunction=/Game/Packages/Fortress_SharedMaterials/Base_Material_Functions/CharacterDeRez_Transition}.Result
  ) {materialFunction=/Game/Packages/Fortress_SharedMaterials/Base_Material_Functions/BaseCharacterShaderFunctions/MF_CosmeticEmissiveEyeAdaptation}
  n19 = Distance(A=CameraPositionWS(), B=ObjectPositionWS())
  n20 = ScalarParameter() {defaultValue=50, group=OverlayOutlines, parameterName=FadeOffset, sliderMax=400, sortPriority=5}
  n21 = Divide(
   A=NamedRerouteDeclaration(None=Subtract(A=@n19, B=@n20) {constB=200}) {name=DistanceAfterOffset, nodeColor=<rgba 0.9631 1 0 1>}
   B=ScalarParameter() {defaultValue=500, group=OverlayOutlines, parameterName=FadeDistance, sliderMax=2000, sortPriority=5}
  ) {constB=1000}
  n22 = InverseLinearInterpolate(Value=MaterialFunctionCall() {materialFunction=/Engine/Functions/Engine_MaterialFunctions02/Utility/FOV}.FOV angle in degrees) {constA=15, constB=90}
  NamedRerouteDeclaration(None=Saturate(None=Min(A=@n21, B=@n22))) {name=DistanceFading, nodeColor=<rgba 1 0.3047 0 1>}
  n23 = ObjectPositionWS()
  n24 = WorldPosition()
  n25 = Saturate(None=@n4)
  n26 = StaticSwitchParameter(True=LinearInterpolate(A=Subtract(A=@n23, B=@n24.XYZ), B=@n9, Alpha=Round(None=@n25)), False=@n9) {group=OverlayOutlines, parameterName=UseVertexOutlineHiding, sortPriority=0}
  NamedRerouteDeclaration(None=StaticSwitchParameter(
    True=MaterialFunctionCall!(WPO=@n26) {materialFunction=/Game/Packages/Fortress_SharedMaterials/Base_Material_Functions/Fortnite_HideElementsWithVertexColor}.Result
    False=@n26
   ) {group=Opacity Mask, parameterName=Use Vertex Colors for Mask, sortPriority=0}) {name=WPO, nodeColor=<rgba 1 0.3281 0 1>}
  n27 = TextureSampleParameter2D() {
   group=OverlayOutlines
   parameterName=OutlineMaskTexture
   samplerType=SAMPLERTYPE_Masks
   sortPriority=1
   texture=&/T_Outline_Mask
  }
  n28 = ChannelMaskParameter(None=@n27.RGB) {group=OverlayOutlines, parameterName=OutlineMaskChannelSelector, sortPriority=3}
  n29 = OneMinus(None=@n28.R)
  n30 = TextureSampleParameter2D() {
   group=OverlayOutlines
   parameterName=SkinFX
   samplerType=SAMPLERTYPE_Masks
   sortPriority=1
   texture=/Game/Global/Textures/Default/Blanks/T_MidgreyMask
  }
  n31 = NamedRerouteDeclaration(None=Add(
    A=ChannelMaskParameter(None=@n30.RGBA) {group=OverlayOutlines, parameterName=OutlineThickness_Channel, sortPriority=3}.R
    B=Constant() {r=-0.009}
   )) {name=ThicknessMask, nodeColor=<rgba 1 0.1689 0 1>}
  n32 = OneMinus(None=TwoSidedSign())
  n33 = StaticSwitchParameter(True=Multiply(A=Ceil(None=@n31) {desc=cut outlines mask}, B=@n32), False=@n32) {group=OverlayOutlines, parameterName=UseSkinFX_Thickness, sortPriority=0}
  n34 = StaticSwitchParameter(
   True=Multiply(A=StaticSwitchParameter(True=@n28.R, False=@n29) {group=OverlayOutlines, parameterName=InvertOutlineMask, sortPriority=0}, B=@n33)
   False=@n33
  ) {
   desc=Exclude a section of the mesh from having Distance Outlines based on a Texture
   group=OverlayOutlines
   parameterName=MaskOutlineByTexture
   sortPriority=0
  }
  NamedRerouteDeclaration(None=Multiply(A=@n34, B=ScalarParameter() {defaultValue=1, group=OverlayOutlines, parameterName=OutlineVisibility, sliderMax=1}) {desc=Controls overall outline visibility}) {name=OpacityMask, nodeColor=<rgba 0 0.8125 1 1>}
 ]
}
