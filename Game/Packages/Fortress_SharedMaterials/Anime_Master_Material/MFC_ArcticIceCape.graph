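# Material function for the Arctic Ice Talus cape backpack that adds a dissolve and wind-driven cloth motion to incoming material attributes, gated by UseArcticIceCape.
# Packed mask texture and bounding-box gradients build dissolve-in and dissolve-out opacity masks, with MaterialCurveB driving progress and MaterialCurveC selecting between them.
# World position offset combines sphere-masked, panned-noise wind scaled by MaterialCurveA with inflation from a moving sphere mask.
# Static switches provide looping dissolve and emissive wind debug views.
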MaterialFunction /Game/Packages/Fortress_SharedMaterials/Anime_Master_Material/MFC_ArcticIceCape
graph {nodes [
  Constant()
  FunctionOutput(None=StaticSwitchParameter(True=NamedRerouteUsage(), False=NamedRerouteUsage()) {group=Skin FX, parameterName=UseArcticIceCape})
  n1 = ScalarParameter() {defaultValue=0.95, group=ArcticIce_Cape, parameterName=WindMask_SmoothStep_Min}
  n2 = ScalarParameter() {defaultValue=0.5, group=ArcticIce_Cape, parameterName=WindMask_SmoothStep_Max}
  n3 = MaterialFunctionCall() {materialFunction=&BoundingBoxBased_0-1_UVW}
  Multiply(A=Saturate(None=SmoothStep(Min=@n1, Max=@n2, Value=@n3.B) {constMax=0.3, constMin=0.95}), B=OneMinus(None=@n3.B)) {constB=0.1}
  n4 = OneMinus(None=NamedRerouteUsage())
  n5 = Add(A=@n4, B=ScalarParameter() {defaultValue=0.101, parameterName=Width})
  n6 = Constant()
  n7 = Constant() {r=1}
  n8 = ScalarParameter() {defaultValue=0.1, group=ArcticIce_Cape, parameterName=DissolveHardness}
  MaterialExpressionRemap_0 = MaterialXRemap!(None=NamedRerouteUsage(), InputLow=@n6, InputHigh=@n7, TargetLow=@n6, TargetHigh=Add(A=@n7, B=@n8))
  Multiply(A=OneMinus(None=SmoothStep(Min=@n5, Max=@n4, Value=@MaterialExpressionRemap_0)), B=Constant3Vector() {constant=<rgba 1 0 0 0>})
  n9 = Frac(None=Multiply(A=Time(), B=ScalarParameter() {defaultValue=1, group=ArcticIce_Cape, parameterName=DebugSpeed}))
  n10 = ScalarParameter() {desc=Dissolve Progress, group=ArcticIce_Cape, parameterName=MaterialCurveB}
  NamedRerouteDeclaration(None=Clamp(None=StaticSwitchParameter(True=@n9, False=@n10) {parameterName=DebugDissolve})) {name=_time, nodeColor=<rgba 0.1719 0 1 1>}
  n11 = GetMaterialAttributes(None=NamedRerouteDeclaration(None=FunctionInput() {inputName=MA, inputType=FunctionInput_MaterialAttributes}) {name=MA_IN, nodeColor=<rgba 1 0 0.9375 1>}) {attributeGetTypes [#1, #2]}
  n12 = Constant()
  n13 = NamedRerouteUsage()
  n14 = ScalarParameter() {defaultValue=1, group=ArcticIce_Cape, parameterName=ArcticIce_Visibility}
  n15 = NamedRerouteUsage()
  NamedRerouteDeclaration(None=SetMaterialAttributes!(
    MaterialAttributes=@n11.MaterialAttributes
    OpacityMask=LinearInterpolate(A=@n12, B=@n13, Alpha=@n14)
    WorldPositionOffset=NamedRerouteUsage()
    EmissiveColor=StaticSwitchParameter(True=@n15, False=@n11.EmissiveColor) {group=ArcticIce_Cape, parameterName=DebugWindWPO}
   ) {attributeSetTypes [#1, #3, #2]}) {name=MA_OUT, nodeColor=<rgba 0 1 0.7656 1>}
  n16 = TextureObjectParameter() {
   group=ArcticIce_Cape
   parameterName=ArcticeIceCape_TA
   samplerType=SAMPLERTYPE_Masks
   texture=/Game/Accessories/FORT_Backpacks/Backpack_M_MED_ArcticIceTalus_Cape/Textures/T_M_Med_ArcticIceTalus_Cape_TA
  }
  n17 = TextureSample(
   UVs=Multiply(A=TextureCoordinate(), B=ScalarParameter() {defaultValue=1, group=ArcticIce_Cape, parameterName=DissolveTex_Tiling})
   Tex=NamedRerouteDeclaration(None=@n16) {name=TA_Tex, nodeColor=<rgba 0 1 0.3906 1>}
  ) {
   samplerType=SAMPLERTYPE_Masks
   texture=&T_M_Med_ArcticIceTalus_Cape_TA
  }
  NamedRerouteDeclaration(None=@n17.B) {name=WPO_Offset, nodeColor=<rgba 1 0 0.4453 1>}
  NamedRerouteDeclaration(None=SmoothStep(
    Min=ScalarParameter() {defaultValue=0.9, group=ArcticIce_Cape, parameterName=Talus_WindMask_SmoothStep_Min}
    Max=ScalarParameter() {group=ArcticIce_Cape, parameterName=Talus_WindMask_SmoothStep_Max}
    Value=MaterialFunctionCall() {materialFunction=&BoundingBoxBased_0-1_UVW}.G
   ) {constMax=0.3, constMin=0.95}) {name=Cape_WindMask_G, nodeColor=<rgba 0.3359 0 1 1>}
  NamedRerouteDeclaration(None=SmoothStep(Min=Add(A=@n4, B=@n8) {constB=0.1}, Max=@n4, Value=@MaterialExpressionRemap_0)) {name=DissolveOut, nodeColor=<rgba 0 1 0.3438 1>}
  n18 = MaterialFunctionCall() {materialFunction=&BoundingBoxBased_0-1_UVW}
  n19 = NamedRerouteUsage()
  MaterialExpressionRemap_2 = MaterialXRemap(None=Subtract(A=OneMinus(None=@n18.B), B=OneMinus(None=@n19)), InputLow=Constant() {r=-1}, InputHigh=Constant() {r=1}) {targetHighDefault=4.5, targetLowDefault=-5}
  n20 = Add(A=NamedRerouteUsage(), B=@MaterialExpressionRemap_2)
  NamedRerouteDeclaration(None=OneMinus(None=SmoothStep(Value=@n20) {constMax=0.05})) {name=DissolveIn, nodeColor=<rgba 0.9688 0 1 1>}
  NamedRerouteDeclaration(None=PreSkinnedPosition()) {name=Preskin_LocalPos, nodeColor=<rgba 0.4531 1 0 1>}
  NamedRerouteDeclaration(None=@n17.R) {name=PerlinNoise, nodeColor=<rgba 1 0.2578 0 1>}
  n21 = SphereMask(A=NamedRerouteUsage(), B=VectorParameter() {group=ArcticIce_Cape, parameterName=WindSphereMaskCenter, defaultValue=<rgba 0 0 20 1>}.RGB) {attenuationRadius=150, hardnessPercent=0}
  NamedRerouteDeclaration(None=Multiply(A=SmoothStep(Value=@n21), B=ScalarParameter() {defaultValue=0.3, group=ArcticIce_Cape, parameterName=WindGradient_Multi}) {constB=0.3}) {name=Cape_WindMask, nodeColor=<rgba 0.3359 0 1 1>}
  n22 = Divide(A=NamedRerouteUsage()) {constB=128}
  n23 = VectorParameter() {group=ArcticIce_Cape, parameterName=WindSpeed, channelNames {a="", b="", g=SpeedY, r=""}, defaultValue=<rgba 0 1.2 0 1>}
  n24 = TextureSample(
   UVs=Panner(Coordinate=ComponentMask(None=@n22) {b=true, r=true}, Speed=ComponentMask(None=@n23) {g=true, r=true}) {speedY=1}
   Tex=NamedRerouteUsage()
  ) {samplerSource=SSM_Wrap_WorldGroupSettings, samplerType=SAMPLERTYPE_LinearGrayscale}
  n25 = VertexNormalWS()
  n26 = Multiply(A=Normalize(VectorInput=@n25), B=ScalarParameter() {defaultValue=8, group=ArcticIce_Cape, parameterName=Wind_VertexNormal_WPO_Max})
  n27 = Normalize(VectorInput=VectorParameter() {group=ArcticIce_Cape, parameterName=FakeWindDirection, defaultValue=<rgba 0 -1.5 0.5 0>}.RGB)
  n28 = Normalize(VectorInput=LinearInterpolate(
    A=@n25
    B=Transform(None=@n27) {transformSourceType=TRANSFORMSOURCE_Local}
    Alpha=ScalarParameter() {defaultValue=0.7, group=ArcticIce_Cape, parameterName=VertexNormal_WindDir_Blend}
   ) {constAlpha=1})
  n29 = Multiply(
   A=Multiply(A=@n24.R, B=NamedRerouteDeclaration(None=@n28) {name=VertexNormal, nodeColor=<rgba 0.6406 0 1 1>})
   B=ScalarParameter() {defaultValue=50, group=ArcticIce_Cape, parameterName=Wind_WPO_Max}
  ) {constB=20}
  n30 = StaticSwitchParameter(True=LinearInterpolate(A=Multiply(A=@n24.R, B=@n26), B=@n29, Alpha=NamedRerouteUsage()), False=@n29) {group=ArcticIce_Cape, parameterName=UseCapeYGradientControl}
  n31 = NamedRerouteUsage()
  n32 = ScalarParameter() {group=ArcticIce_Cape, parameterName=MaterialCurveA}
  NamedRerouteDeclaration(None=Multiply(A=Multiply(A=@n30, B=@n31), B=Clamp(None=@n32))) {name=Wind_WPO, nodeColor=<rgba 0.0547 1 0 1>}
  NamedRerouteDeclaration(None=Add(A=NamedRerouteUsage(), B=NamedRerouteUsage())) {name=Final_WPO, nodeColor=<rgba 1 0 0.2344 1>}
  n33 = NamedRerouteUsage()
  n34 = SphereMask(
   A=NamedRerouteUsage()
   B=LinearInterpolate(
    A=VectorParameter() {parameterName=InflationMask_Center_Start, defaultValue=<rgba 0 0 30 1>}.RGB
    B=VectorParameter() {parameterName=InflationMask_Center_End, defaultValue=<rgba 0 0 100 1>}.RGB
    Alpha=@n33
   ) {constA=50}
   Radius=ScalarParameter() {defaultValue=90, group=ArcticIce_Cape, parameterName=Inflation_Radius}
  ) {attenuationRadius=80, hardnessPercent=0}
  n35 = ScalarParameter() {group=ArcticIce_Cape, parameterName=Inflation_WPO_Min}
  n36 = ScalarParameter() {defaultValue=70, group=ArcticIce_Cape, parameterName=Inflation_WPO_Max}
  n37 = Multiply(A=NamedRerouteUsage(), B=ScalarParameter() {defaultValue=20, parameterName=WPO_Offset_Intensity}) {constB=50}
  n38 = Multiply(A=Normalize(VectorInput=NamedRerouteUsage()), B=Add(A=LinearInterpolate(A=@n35, B=@n36, Alpha=@n33) {constA=50}, B=Multiply(A=@n33, B=@n37))) {constB=20}
  NamedRerouteDeclaration(None=Multiply(A=Multiply(A=@n34, B=@n38), B=ScalarParameter() {defaultValue=1, group=ArcticIce_Cape, parameterName=Inflation_WPO_Intensity, sliderMax=1})) {name=Inflation_WPO, nodeColor=<rgba 0 0.4375 1 1>}
  NamedRerouteDeclaration(None=If(A=ScalarParameter() {group=ArcticIce_Cape, parameterName=MaterialCurveC}, "A > B"=NamedRerouteUsage(), "A < B"=NamedRerouteUsage()) {constB=0.5}) {name=Final_Opacity, nodeColor=<rgba 0 0.6719 1 1>}
  NamedRerouteDeclaration(None=@n17.G) {name=ClothPiecesGradient, nodeColor=<rgba 0.3828 0 1 1>}
  OneMinus(None=@n9)
  ScalarParameter() {group=ArcticIce_Cape, parameterName=Wind_Intensity, sliderMax=1}
 ]}
