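# Material function that adds animated wind ripples to cloth through normal and World Position Offset, toggled by UseClothWind.
# Panning ripple UVs come from bounds-normalized pre-skinned local XY, computed per vertex and interpolated for pixels.
# Ripple textures drive optional blended normals and local-Z displacement, both pulsed by a clamped time sine.
# A selectable FX input channel masks both effects before they rewrite the incoming material attributes.
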MaterialFunction /Game/Packages/Fortress_SharedMaterials/Anime_Master_Material/MFC_ClothWind
graph {nodes [
  FunctionOutput(None=StaticSwitchParameter(True=NamedRerouteUsage(), False=NamedRerouteUsage()) {group=Skin FX, parameterName=UseClothWind})
  NamedRerouteDeclaration(None=FunctionInput() {inputName=Material, inputType=FunctionInput_MaterialAttributes}) {name=MI_in, nodeColor=<rgba 1 0 0.5156 1>}
  n1 = GetMaterialAttributes(None=NamedRerouteUsage()) {attributeGetTypes [#1, #2]}
  n2 = NamedRerouteUsage()
  n3 = NamedRerouteUsage()
  n4 = LinearInterpolate(
   A=@n1.Normal
   B=Normalize(VectorInput=MaterialFunctionCall(BaseNormal=@n1.Normal, AdditionalNormal=@n2) {materialFunction=&BlendAngleCorrectedNormals}.Result)
   Alpha=@n3
  )
  n5 = Add(A=@n1.WorldPositionOffset, B=NamedRerouteUsage())
  NamedRerouteDeclaration(None=SetMaterialAttributes!(
    MaterialAttributes=@n1.MaterialAttributes
    Normal=StaticSwitchParameter(True=@n4, False=@n1.Normal) {group=ClothWind, parameterName=UseNormalAnimation, sortPriority=1}
    WorldPositionOffset=LinearInterpolate(A=@n1.WorldPositionOffset, B=@n5, Alpha=@n3)
   ) {attributeSetTypes [#1, #2]}) {name=MA_out, nodeColor=<rgba 1 0.5859 0 1>}
  n6 = ChannelMaskParameter(None=NamedRerouteDeclaration(None=FunctionInput() {inputName=FX, sortPriority=1}) {name=FX_in, nodeColor=<rgba 0 0.7422 1 1>}) {group=ClothWind, maskChannel=Blue, parameterName=WindMask_Channel, sortPriority=2, defaultValue=<rgba 0 0 1 0>}
  NamedRerouteDeclaration(None=@n6.R) {name=WindMask, nodeColor=<rgba 0.4778 1 0 1>}
  n7 = MaterialFunctionCall(Float3=TextureSample(UVs=NamedRerouteUsage()) {samplerType=SAMPLERTYPE_Normal, texture=/Game/Global/Textures/Default/Noise/T_VerticalRipples_N}.RGB) {materialFunction=&BreakOutFloat3Components}
  n8 = ScalarParameter() {defaultValue=1, group=ClothWind, parameterName=RippleNormalPow}
  n9 = Multiply(A=Time(), B=ScalarParameter() {defaultValue=0.2, group=ClothWind, parameterName=TimeRippleAttenSpeed})
  n10 = Max(A=ConstantBiasScale(None=Sine(None=@n9)), B=ScalarParameter() {defaultValue=0.15, group=ClothWind, parameterName=TimeRippleAttenMin}) {constB=0.15}
  n11 = Multiply(
   A=MaterialFunctionCall(X=@n7.R, Y=Power(Base=@n7.G, Exp=@n8)) {materialFunction=&MakeFloat2}.Result
   B=Multiply(A=ScalarParameter() {defaultValue=0.25, group=ClothWind, parameterName=RippleNormalIntensity}, B=@n10)
  )
  NamedRerouteDeclaration(None=Normalize(VectorInput=AppendVector(A=@n11, B=@n7.B))) {name=Normal, nodeColor=<rgba 0.6641 0 1 1>}
  n12 = ScalarParameter() {defaultValue=10, group=ClothWind, parameterName=RippleWPOIntensity}
  n13 = Constant()
  n14 = MaterialFunctionCall(
   R=@n13
   G=@n13
   B=TextureSample(UVs=NamedRerouteUsage()) {samplerType=SAMPLERTYPE_Masks, texture=/Game/Global/Textures/Default/Noise/T_VerticalRipples_D}.G
  ) {materialFunction=&AppendMany}
  NamedRerouteDeclaration(None=Multiply(A=Multiply(A=@n10, B=@n12), B=Transform(None=@n14.RGB) {transformSourceType=TRANSFORMSOURCE_Local})) {name=WPO, nodeColor=<rgba 0.3594 1 0 1>}
  n15 = PreSkinnedPosition()
  n16 = PreSkinnedLocalBounds()
  n17 = Divide(
   A=Subtract(A=NamedRerouteDeclaration(None=@n15) {name=Pre Skinned Local Postition, nodeColor=<rgba 1 0.5156 0 1>}, B=@n16.Min)
   B=@n16.Extents
  )
  n18 = VectorParameter() {group=ClothWind, parameterName=RippleTiling, defaultValue=<rgba 1 1 0 0>}
  n19 = Panner(
   Coordinate=MaterialFunctionCall(UVs=ComponentMask(None=@n17) {g=true, r=true}, Texture Scale=ComponentMask(None=@n18.RGB) {g=true, r=true}) {materialFunction=&ScaleUVsByCenter}.UVs
   Speed=ComponentMask(None=VectorParameter() {group=ClothWind, parameterName=RippleSpeed, defaultValue=<rgba 0 -1 0 0>}.RGB) {g=true, r=true}
  )
  NamedRerouteDeclaration(None=ShaderStageSwitch(PixelShader=VertexInterpolator(VS=@n19).PS, VertexShader=@n19)) {name=UV Coords, nodeColor=<rgba 0 0.8125 1 1>}
 ]}
