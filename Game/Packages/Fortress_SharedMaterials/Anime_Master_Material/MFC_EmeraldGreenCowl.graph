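# Material function for the anime master material that adds a switchable "Emerald Green Cowl" skin effect by overriding emissive color.
# FX input channels supply cowl masks, including a glow that varies with camera distance.
# Sphere masks on pre-skinned position drive a transition and glowing band, animated by MaterialCurve parameters.
# Frame-stepped, scrolling crack texture and hash flashing form electricity, and the shading-curve input supports toon-style shadow darkening.
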MaterialFunction /Game/Packages/Fortress_SharedMaterials/Anime_Master_Material/MFC_EmeraldGreenCowl
graph {nodes [
  n1 = FunctionInput() {inputName=MA, inputType=FunctionInput_MaterialAttributes}
  FunctionOutput(None=StaticSwitchParameter(True=NamedRerouteUsage(), False=NamedRerouteDeclaration(None=@n1) {name=Material, nodeColor=<rgba 0 1 0.8125 1>}) {group=Skin FX, parameterName=UseEmeraldGreenCowl})
  n2 = FunctionInput() {bUsePreviewValueAsDefault=true, inputName=FX, inputType=FunctionInput_Vector4, sortPriority=20}
  n3 = ChannelMaskParameter(None=@n2) {group=Emerald Green Cowl, maskChannel=Green, parameterName=EGC FX Mask Blur Channel, defaultValue=<rgba 0 1 0 0>}
  n4 = LinearInterpolate(
   A=ScalarParameter() {defaultValue=2, group=Emerald Green Cowl, parameterName=EGC Glow Intensity Far}
   B=ScalarParameter() {defaultValue=1, group=Emerald Green Cowl, parameterName=EGC Glow Intensity Close}
   Alpha=SphereMask(A=WorldPosition().XYZ, B=CameraPositionWS(), Radius=ScalarParameter() {defaultValue=1000, parameterName=EGC Glow Intentisy Fade Distance}) {hardnessPercent=0}
  )
  NamedRerouteDeclaration(None=Power(Base=Multiply(A=@n3.R, B=@n4), Exp=ScalarParameter() {defaultValue=1.3, group=Emerald Green Cowl, parameterName=EGC Glow Exponent})) {name=Mask2, nodeColor=<rgba 0 0.6016 1 1>}
  NamedRerouteDeclaration(None=ChannelMaskParameter(None=@n2) {group=Emerald Green Cowl, parameterName=EGC FX Mask Channel}.R) {name=Mask, nodeColor=<rgba 0 0.6016 1 1>}
  n5 = VertexInterpolator(VS=PreSkinnedPosition())
  n6 = VectorParameter() {group=Emerald Green Cowl, parameterName=EGCA Transition Pivot, defaultValue=<rgba 0 0 96.549 1>}
  n7 = ScalarParameter() {defaultValue=30, group=Emerald Green Cowl, parameterName=EGCA Glow Width}
  n8 = LinearInterpolate(
   A=ScalarParameter() {defaultValue=220, group=Emerald Green Cowl, parameterName=EGCA Radius}
   B=Multiply(A=@n7) {constB=-1}
   Alpha=Max(
    A=ScalarParameter() {defaultValue=1, group=Emerald Green Cowl, parameterName=MaterialCurveA, sliderMax=1}
    B=ScalarParameter() {parameterName=EGCA Disable Animation}
   )
  ) {constB=0}
  n9 = ScalarParameter() {defaultValue=0.904, group=Emerald Green Cowl, parameterName=EGCA Hardness, sliderMax=1}
  NamedRerouteDeclaration(None=OneMinus(None=SphereMask(A=@n5.PS, B=@n6.RGB, Radius=@n8, Hardness=@n9))) {name=Animation Mask, nodeColor=<rgba 1 0 0.1406 1>}
  n10 = GetMaterialAttributes(None=NamedRerouteUsage()) {attributeGetTypes [#1]}
  n11 = VectorParameter() {group=Emerald Green Cowl, parameterName=EGC Cowl Primary Color, defaultValue=<rgba 0.5029 0.0194 0.1144 1>}
  n12 = NamedRerouteUsage()
  n13 = ScalarParameter() {defaultValue=15, group=Emerald Green Cowl, parameterName=EGC Electricity FPS}
  n14 = Floor(None=Multiply(A=Time(), B=@n13))
  n15 = VectorParameter() {group=Emerald Green Cowl, parameterName=EGC Electricity Flow Speed, defaultValue=<rgba 264 145.6 -298.2 0>}
  n16 = ScalarParameter() {defaultValue=1, group=Emerald Green Cowl, parameterName=EGC Electricity Global Scale}
  n17 = Multiply(A=Multiply(A=Divide(A=@n14, B=@n13), B=@n15.RGB), B=@n16)
  n18 = TextureCoordinate()
  n19 = Add(
   A=Frac(None=ComponentMask(None=@n17) {g=true, r=true})
   B=Multiply(A=ScalarParameter() {defaultValue=20, group=Emerald Green Cowl, parameterName=EGC Electricity Texture Scale 1}, B=Multiply(A=@n16, B=@n18))
  )
  n20 = TextureObjectParameter() {
   group=Emerald Green Cowl
   parameterName=EGC Electricity Texture
   samplerType=SAMPLERTYPE_Grayscale
   texture=/Game/Effects/Fort_Effects/Textures/Tile/T_Concrete_Crack_A
  }
  n21 = Max(
   A=NamedRerouteDeclaration(None=TextureSample(UVs=@n19, Tex=@n20).R) {name=Electricity Mask, nodeColor=<rgba 1 0 0.0469 1>}
   B=Multiply(
    A=MaterialFunctionCall() {materialFunction=&MF_Hash}.Float
    B=ScalarParameter() {defaultValue=0.1, group=Emerald Green Cowl, parameterName=EGC Flashing Intensity}
   )
  )
  n22 = VectorParameter() {group=Emerald Green Cowl, parameterName=EGC Cowl Transition Color, defaultValue=<rgba 3 0.3281 0.1875 1>}
  n23 = VectorParameter() {group=Emerald Green Cowl, parameterName=EGC Cowl Transition Color Core, defaultValue=<rgba 6 5.0328 2.1975 1>}
  n24 = OneMinus(None=NamedRerouteUsage())
  n25 = NamedRerouteUsage()
  n26 = LinearInterpolate(
   A=LinearInterpolate(
    A=@n11.RGB
    B=VectorParameter() {group=Emerald Green Cowl, parameterName=EGC Cowl Noise Color, defaultValue=<rgba 3 0.3616 2.4932 1>}.RGB
    Alpha=Multiply(A=@n12, B=@n21)
   )
   B=LinearInterpolate(A=@n22.RGB, B=@n23.RGB, Alpha=Min(A=@n24, B=@n25))
   Alpha=Multiply(A=NamedRerouteUsage(), B=ScalarParameter() {defaultValue=1, group=Emerald Green Cowl, parameterName=MaterialCurveB, sliderMax=1})
  )
  n27 = LinearInterpolate(
   A=Multiply(A=@n26, B=ScalarParameter() {defaultValue=0.75, group=Emerald Green Cowl, parameterName=EGC Shadow Multiplier})
   B=@n26
   Alpha=NamedRerouteUsage()
  )
  n28 = Max(A=NamedRerouteUsage(), B=@n12)
  n29 = Multiply(A=NamedRerouteUsage(), B=NamedRerouteDeclaration(None=Saturate(None=@n28)) {name=Mask3, nodeColor=<rgba 1 0.6094 0 1>})
  NamedRerouteDeclaration(None=SetMaterialAttributes!(MaterialAttributes=@n10.MaterialAttributes, EmissiveColor=LinearInterpolate(A=@n10.EmissiveColor, B=@n27, Alpha=@n29)) {attributeSetTypes [#1]}) {name=Material Out, nodeColor=<rgba 1 0 0.5625 1>}
  NamedRerouteDeclaration(None=FunctionInput() {bUsePreviewValueAsDefault=true, inputName=SC, inputType=FunctionInput_Scalar, sortPriority=10}) {name=Shading Curve, nodeColor=<rgba 0.8516 0 1 1>}
  NamedRerouteDeclaration(None=SphereMask(
    A=@n5.PS
    B=@n6.RGB
    Radius=Add(A=@n8, B=@n7)
    Hardness=ScalarParameter() {group=Emerald Green Cowl, parameterName=EGCA Glow Hardness, sliderMax=1}
   )) {name=Animation Glow Mask, nodeColor=<rgba 1 0 0.1406 1>}
 ]}
