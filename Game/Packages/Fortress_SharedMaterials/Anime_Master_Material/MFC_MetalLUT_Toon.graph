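# Material function for the anime master material that gives masked areas a toon metal look by tinting Emissive Color with a reflectance LUT.
# The clamped camera-vector and pixel-normal dot product samples a curve atlas row, defaulting to an aluminium reflectance curve, for a view-angle Fresnel response.
# That result multiplies the incoming emissive, blended by a selectable FX input channel, green by default.
# The MetalLUT_Toon static switch otherwise passes the material attributes through unchanged.
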MaterialFunction /Game/Packages/Fortress_SharedMaterials/Anime_Master_Material/MFC_MetalLUT_Toon
graph {nodes [
  n1 = FunctionInput() {inputName=MA, inputType=FunctionInput_MaterialAttributes}
  n2 = GetMaterialAttributes(None=@n1) {attributeGetTypes [#1]}
  n3 = CameraVectorWS()
  n4 = PixelNormalWS()
  n5 = CurveAtlasRowParameter(CurveTime=Clamp(None=DotProduct(A=@n3, B=@n4))) {
   atlas=&AthenaCurveAtlas01
   curve=/Game/Characters/Player/Female/Medium/Bodies/F_MED_Ms_Alpine/Textures/AluminiumreflectanceCurve
   defaultValue=189
   group=MetalLUTToon
   parameterName=FresnelReflectance_LUT
  }
  n6 = ChannelMaskParameter(None=FunctionInput() {inputName=FX, inputType=FunctionInput_Vector4, sortPriority=1}) {group=MetalLUTToon, maskChannel=Green, parameterName=MetalLUT_Toon_Channel, defaultValue=<rgba 0 1 0 0>}
  n7 = LinearInterpolate(A=@n2.EmissiveColor, B=Multiply(A=@n2.EmissiveColor, B=@n5), Alpha=@n6.R)
  FunctionOutput(None=StaticSwitchParameter(
    True=SetMaterialAttributes!(MaterialAttributes=@n2.MaterialAttributes, EmissiveColor=@n7) {attributeSetTypes [#1]}
    False=@n1
   ) {group=Skin FX, parameterName=MetalLUT_Toon})
 ]}
