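# Toon material function that adds a thin-film iridescent sheen to incoming material attributes as an optional Skin FX, toggled by UseThinFilmToon.
# A Fresnel-warped coordinate, with a static offset or modulo-time animation, samples a spectrum texture scaled by intensity.
# The color is added to emissive, using channel-selected SSC and FX masks and a second Fresnel gradient to control the blend.
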MaterialFunction /Game/Packages/Fortress_SharedMaterials/Anime_Master_Material/MFC_ThinFilm_Toon
graph {nodes [
  FunctionOutput(None=StaticSwitchParameter(True=NamedRerouteUsage(), False=NamedRerouteUsage()) {group=Skin FX, parameterName=UseThinFilmToon})
  NamedRerouteDeclaration(None=FunctionInput() {inputName=MA, inputType=FunctionInput_MaterialAttributes}) {name=MA_In, nodeColor=<rgba 1 0 0.375 1>}
  NamedRerouteDeclaration(None=ChannelMaskParameter(None=FunctionInput() {inputName=FX, inputType=FunctionInput_Vector4, sortPriority=1, previewValue=<1 0 0 1>}) {group=ThinFilm_Toon, parameterName=SkinFX_Channel, sortPriority=10}.R) {name=SSC_Sheen, nodeColor=<rgba 0 1 0.1562 1>}
  n1 = GetMaterialAttributes(None=NamedRerouteUsage()) {attributeGetTypes [#1]}
  n2 = LinearInterpolate(A=@n1.EmissiveColor, B=Add(A=@n1.EmissiveColor, B=NamedRerouteUsage()), Alpha=NamedRerouteUsage())
  NamedRerouteDeclaration(None=BlendMaterialAttributes(
    A=NamedRerouteUsage()
    B=SetMaterialAttributes!(MaterialAttributes=@n1.MaterialAttributes, EmissiveColor=@n2) {attributeSetTypes [#1]}
    Alpha=NamedRerouteUsage()
   ) {vertexAttributeBlendType=UseA}) {name=MA_Out, nodeColor=<rgba 0 0.1328 1 1>}
  n3 = MaterialFunctionCall!(
   Time
   Divisor=ScalarParameter() {defaultValue=1, group=ThinFilm_Toon, parameterName=ThinFilm_Animated_TimeDivisor, sliderMax=2}
   Mod=ScalarParameter() {defaultValue=1, group=ThinFilm_Toon, parameterName=ThinFilm_Animated_TimeModulo, sliderMax=2}
  ) {materialFunction=/Game/Effects/Fort_Effects/Materials/_Functions/MF_ModuloTime}
  n4 = ScalarParameter() {group=ThinFilm_Toon, parameterName=ThinFilm_Offset, sliderMax=1, sliderMin=-1}
  n5 = Fresnel(ExponentIn=ScalarParameter() {defaultValue=0.5, group=ThinFilm_Toon, parameterName=ThinFilm_Warp, sliderMax=4}) {baseReflectFraction=0}
  n6 = ScalarParameter() {defaultValue=1, group=ThinFilm_Toon, parameterName=ThinFilm_Scale, sliderMax=2}
  n7 = TextureSampleParameter2D(UVs=Add(
    A=StaticSwitchParameter(True=@n3.Unclamped, False=@n4) {group=ThinFilm_Toon, parameterName=AnimateThinFilm, sortPriority=1}
    B=Multiply(A=@n5, B=@n6)
   )) {
   group=ThinFilm_Toon
   parameterName=Thin Film Texture
   sortPriority=5
   texture=/Game/Global/Textures/Default/Thinfilm/T_ThinFilm_Spectrum_COLOR
  }
  NamedRerouteDeclaration(None=Multiply(A=@n7.RGB, B=ScalarParameter() {defaultValue=0.5, group=ThinFilm_Toon, parameterName=ThinFilm_Intensity, sliderMax=2})) {name=ThinFilm_Scaled, nodeColor=<rgba 1 0.9375 0 1>}
  n8 = ChannelMaskParameter(None=FunctionInput() {inputName=SSC, inputType=FunctionInput_Vector4, sortPriority=2, previewValue=<0 0 1 1>}) {group=ThinFilm_Toon, maskChannel=Blue, parameterName=SSC_Channel, sortPriority=10, defaultValue=<rgba 0 0 1 0>}
  NamedRerouteDeclaration(None=@n8.R) {name=Sheen, nodeColor=<rgba 0 0 1 1>}
  n9 = Multiply(A=NamedRerouteUsage(), B=ScalarParameter() {defaultValue=1, group=ThinFilm_Toon, parameterName=ThinFilm_SheenScale, sliderMax=2})
  n10 = ScalarParameter() {defaultValue=1.2, group=ThinFilm_Toon, parameterName=ThinFilm_Exponent, sliderMax=4}
  NamedRerouteDeclaration(None=Multiply(A=Saturate(None=@n9), B=Fresnel(ExponentIn=@n10) {baseReflectFraction=0})) {name=FresnelGradient, nodeColor=<rgba 0.0547 0 1 1>}
 ]}
