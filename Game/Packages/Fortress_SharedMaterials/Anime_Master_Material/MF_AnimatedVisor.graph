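# Material function in the anime master material set that adds a tinted, optionally animated visor glow to incoming material attributes.
# A component-masked EmissiveMask times VisorColor is multiplied by a distance-field SubUV flipbook unless RemoveAnimation is on, then added to EmissiveColor.
# A static switch toggles the effect.
# A mobile toggle chooses between a direct reroute and that reroute through a ShadingPathSwitch with only Default wired.
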MaterialFunction /Game/Packages/Fortress_SharedMaterials/Anime_Master_Material/MF_AnimatedVisor
graph {
 properties {description=10}
 nodes [
  n1 = NamedRerouteUsage()
  n2 = ShadingPathSwitch(Default=@n1) {desc=Migrated from FeatureLevelSwitch}
  n3 = StaticBoolParameter() {defaultValue=true, group=Visor, parameterName=UseAnimatedViroOnMobile, sortPriority=0}
  FunctionOutput(None=StaticSwitchParameter(True=StaticSwitch(True=@n1, False=@n2, Value=@n3), False=NamedRerouteUsage()) {group=Skin FX, parameterName=UseAnmimatedVisor})
  NamedRerouteDeclaration(None=FunctionInput() {inputType=FunctionInput_MaterialAttributes}) {name=MA, nodeColor=<rgba 0 1 0.4375 1>}
  n4 = GetMaterialAttributes(None=NamedRerouteUsage()) {attributeGetTypes [#1, #2]}
  n5 = Multiply(
   A=StaticComponentMaskParameter(None=TextureSampleParameter2D() {group=Visor, parameterName=EmissiveMask, sortPriority=1, texture=&T_BlacksRGB}.RGB) {defaultB=true, defaultG=true, defaultR=true, group=Visor, parameterName=Emissive Component Mask, sortPriority=2}
   B=VectorParameter() {group=Visor, parameterName=VisorColor, sortPriority=3, defaultValue=<rgba 1 0 0.0477 1>}.RGB
  )
  n6 = ScalarParameter() {defaultValue=2, group=Visor, parameterName=SubUV_Frames}
  n7 = TextureObjectParameter() {
   group=Visor
   parameterName=DF Texture
   samplerType=SAMPLERTYPE_LinearGrayscale
   sortPriority=1
   texture=/Game/Global/Textures/Default/Blanks/T_BlackLinearGrayscale
  }
  n8 = ScalarParameter() {defaultValue=1, group=Visor, parameterName=SubUV_Speed}
  n9 = StaticSwitchParameter(
   True=@n5
   False=Multiply(
    A=@n5
    B=MaterialFunctionCall!(SubUV_Frames=@n6, SubUV_TextureObject=@n7, SubUV_Speed=@n8) {materialFunction=/Game/Packages/Fortress_SharedMaterials/Base_Material_Functions/MF_DistanceFieldSubUV}.Result
   )
  ) {group=Visor, parameterName=RemoveAnimation, sortPriority=0}
  NamedRerouteDeclaration(None=SetMaterialAttributes!(MaterialAttributes=@n4.MaterialAttributes, EmissiveColor=Add(A=@n4.EmissiveColor, B=@n9)) {attributeSetTypes [#2]}) {name=AnimatedVisor, nodeColor=<rgba 0 0.5312 1 1>}
  NamedRerouteUsage()
 ]
}
