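# Anime master material function that fades the roughness input by object orientation relative to the camera, apparently to control lines stored in the roughness channel.
# A cross product built from object orientation and camera-to-object vectors drives either a simple adjusted fade or a sweep.
# The sweep adds a pre-skinned-position gradient computed through a vertex interpolator.
# Static switches toggle the effect, sweep mode, channel-selected texture mask and a debug view of the fade.
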MaterialFunction /Game/Packages/Fortress_SharedMaterials/Anime_Master_Material/MF_OrientationFade
graph {nodes [
  n1 = TextureSampleParameter2D() {
   group=Line Orientation Fade
   parameterName=Mask for Fade Orientation
   samplerType=SAMPLERTYPE_Masks
   texture=&T_BlackMask
  }
  n2 = StaticSwitchParameter(True=ChannelMaskParameter(None=@n1.RGB) {group=RoughnessGBufferLines, parameterName=Channel Select For Fade Mask}.R, False=Constant() {r=1}) {group=Line Orientation Fade, parameterName=Use Mask on Fade Orientation}
  n3 = Constant3Vector()
  n4 = ComponentMask(None=Subtract(A=PreSkinnedPosition(), B=@n3)) {g=true, r=true}
  n5 = DotProduct(A=Constant3Vector() {constant=<rgba 0 1 0 0>}, B=Normalize(VectorInput=AppendVector(A=@n4, B=@n3.R)))
  n6 = VectorParameter() {
   group=Line Orientation Fade
   parameterName=SweepValues
   channelNames {a=DarknessMultiplier, b=Subtract, g=Multiply, r=Power}
   defaultValue=<rgba 3 1 0 2>
  }
  n7 = Saturate(None=Multiply(A=Power(Base=@n5, Exp=@n6.Power) {constExponent=5}, B=@n6.Multiply))
  n8 = Add(A=Saturate(None=VertexInterpolator(VS=@n7).PS))
  n9 = MaterialFunctionCall(Float3=TransformPosition(None=ObjectOrientation()) {transformSourceType=TRANSFORMPOSSOURCE_World}) {materialFunction=&BreakOutFloat3Components}
  n10 = Constant()
  n11 = MaterialFunctionCall(X=@n9.R, Y=@n9.R, Z=@n10) {materialFunction=&MakeFloat3}
  n12 = ObjectPositionWS()
  n13 = CameraPositionWS()
  n14 = MaterialFunctionCall(Float3=TransformPosition(None=Subtract(A=@n12, B=@n13)) {transformSourceType=TRANSFORMPOSSOURCE_World}) {materialFunction=&BreakOutFloat3Components}
  n15 = MaterialFunctionCall(X=@n14.R, Y=@n14.R, Z=@n10) {materialFunction=&MakeFloat3}
  n16 = CrossProduct(A=Normalize(VectorInput=Subtract(A=@n11.Result, B=@n15.Result)), B=Constant3Vector() {constant=<rgba 0 0 1 0>})
  n17 = ComponentMask(None=@n16) {r=true}
  n18 = Subtract(A=Constant() {r=1}, B=Multiply(A=Subtract(A=@n8, B=@n6.Subtract), B=OneMinus(None=@n17)))
  n19 = Multiply(
   A=Add(A=@n16, B=ScalarParameter() {defaultValue=-0.1, group=Line Orientation Fade, parameterName=Add Adjust for Orientation Fade}) {constB=-0.1}
   B=ScalarParameter() {defaultValue=0.75, group=Line Orientation Fade, parameterName=Multiply Adjust for Orientation Fade}
  ) {constB=0.75}
  n20 = Multiply(
   A=StaticSwitchParameter(True=Saturate(None=@n18), False=Saturate(None=@n19)) {group=Line Orientation Fade, parameterName=UseSweepInFade}
   B=@n6.DarknessMultiplier
  )
  n21 = ComponentMask(None=Multiply(A=@n2, B=Saturate(None=@n20))) {r=true}
  n22 = FunctionInput() {inputName=Roughness input}
  n23 = LinearInterpolate(A=@n22, B=Multiply(A=@n22, B=@n21), Alpha=@n2)
  FunctionOutput(None=StaticSwitchParameter(True=StaticSwitchParameter(True=@n21, False=@n23) {group=Line Orientation Fade, parameterName=DEBUG VIEW}, False=@n22) {group=Line Orientation Fade, parameterName=Fade Roughness by Orientation})
 ]}
