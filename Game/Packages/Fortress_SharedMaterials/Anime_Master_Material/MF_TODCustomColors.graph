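# Material function for the Anime master material that tints an input color by time of day.
# Four vector parameters for morning, day, evening and night feed the Color_Time_of_Day function, whose result multiplies the input.
# The UseTODCustomColors static switch passes the input through unchanged when off.
# A note marks the custom colors as a temporary fix for burnt colors at night, hence the slightly darker, bluish night default.
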MaterialFunction /Game/Packages/Fortress_SharedMaterials/Anime_Master_Material/MF_TODCustomColors
graph {nodes [
  Constant4Vector() {desc=original, constant=<rgba 0.6 0.6 0.7 1>}
  n1 = MaterialFunctionCall(
   Color_Morning=VectorParameter() {group=Toon_Shading_TODCustomColors, parameterName=Color_Morning, defaultValue=<rgba 1 1 1 1>}.RGB
   Color_Day=VectorParameter() {group=Toon_Shading_TODCustomColors, parameterName=Color_Day, defaultValue=<rgba 1 1 1 1>}.RGB
   Color_Evening=VectorParameter() {group=Toon_Shading_TODCustomColors, parameterName=Color_Evening, defaultValue=<rgba 1 1 1 1>}.RGB
   Color_Night=VectorParameter() {group=Toon_Shading_TODCustomColors, parameterName=Color_Night, defaultValue=<rgba 0.725 0.725 0.85 1>}.RGB
  ) {desc=Temp fix for the burnt colors at night, materialFunction=&Color_Time_of_Day}
  n2 = FunctionInput()
  FunctionOutput(None=StaticSwitchParameter(True=Multiply(A=@n1.Result, B=@n2), False=@n2) {group=Toon_Shading_TODCustomColors, parameterName=UseTODCustomColors, sortPriority=0})
 ]}
