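# Material function for toon/anime character shading that rebuilds lighting in emissive on a Cloth shading model, with a PBR override blend.
# Light vectors (sun, screen-space or camera) and adjusted normals (spherical, painted, shadow-gradient) feed a curve-atlas ramp, masked by SSC thresholds, that blends shaded and lit colors.
# Rim/backlight, Blinn/Phong or anisotropic hair highlights, fake PBR, LUT and time-of-day adjustments layer on top.
# Distance-field ink lines get distance-scaled thickness, and SSC, ink, shading and FX masks are extra outputs.
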MaterialFunction /Game/Packages/Fortress_SharedMaterials/Anime_Master_Material/MF_AnimeShading
graph {
 properties {libraryCategoriesText [Toon]}
 nodes [
  AtmosphericLightVector()
  n1 = NamedRerouteUsage()
  ComponentMask(None=@n1) {b=true, desc=Sheen}
  n2 = VectorParameter() {group=Toon_Shading, parameterName=ScreenSpaceLightVector, defaultValue=<rgba 1 1 -1 1>}
  Composite!(LightVector=@n2.RGB) {subgraphName=ProjectScreenAlighned}
  n3 = NamedRerouteUsage()
  n4 = ScalarParameter() {group=Toon_Shading_PBR_Override, parameterName=PBR_BaseColor_EmissiveIntensity, sliderMax=1, sortPriority=19}
  n5 = NamedRerouteUsage()
  n6 = ScalarParameter() {group=Toon_Shading_PBR_Override, parameterName=PBR_EmissiveIntensity, sliderMax=1, sortPriority=20}
  n7 = Constant()
  n8 = NamedRerouteUsage()
  n9 = LinearInterpolate(A=ScalarParameter() {defaultValue=0.5, group=Toon_Shading, parameterName=ThresholdDarkeningValue, sliderMax=1}, Alpha=NamedRerouteUsage())
  n10 = NamedRerouteUsage()
  n11 = TextureSampleParameter2D(UVs=@n10) {group=Toon_Shading, parameterName=ShadedDiffuse, sortPriority=1, texture=/Game/Global/Textures/Default/Blanks/T_MidgreysRGB}
  n12 = Multiply(A=@n11.RGB, B=ScalarParameter() {defaultValue=1, group=Toon_Shading, parameterName=ShadedColorDarkening, sliderMax=1})
  n13 = FunctionInput(Preview=TextureSampleParameter2D(UVs=@n10) {group=Toon_Shading, parameterName=LitDiffuse, sortPriority=1, texture=&T_WhitesRGB}.RGB) {bUsePreviewValueAsDefault=true, inputName=LitDiffuseOverride, sortPriority=1}
  n14 = Multiply(
   A=VectorParameter() {group=Toon_Shading, parameterName=TintColor, defaultValue=<rgba 0.054 0.054 0.054 1>}.RGB
   B=ScalarParameter() {defaultValue=1, group=Toon_Shading, parameterName=TintColor_Darkening, sliderMax=1}
  )
  n15 = Multiply(
   A=VectorParameter() {group=Toon_Shading, parameterName=TintColor2, defaultValue=<rgba 0.054 0.054 0.054 1>}.RGB
   B=ScalarParameter() {defaultValue=1, group=Toon_Shading, parameterName=TintColor2_Darkening, sliderMax=1}
  )
  n16 = FunctionInput() {bUsePreviewValueAsDefault=true, inputName=Tint Mask, inputType=FunctionInput_Scalar, sortPriority=1}
  n17 = Multiply(
   A=@n13
   B=StaticSwitchParameter(True=LinearInterpolate(A=@n14, B=@n15, Alpha=@n16), False=@n14) {group=Toon_Shading, parameterName=useTintMask, sortPriority=0}
  )
  n18 = MaterialFunctionCall!(LitBaseColorIn=@n13) {materialFunction=/Game/Packages/Fortress_SharedMaterials/Cinematics/Anime/MF_CINE_LitDiffuseTint}
  n19 = LinearInterpolate(
   A=MaterialFunctionCall!(ShadedBaseColorIn=StaticSwitchParameter(True=@n12, False=@n17) {group=Toon_Shading, parameterName=UseShadingTexture, sortPriority=0}) {materialFunction=/Game/Packages/Fortress_SharedMaterials/Cinematics/Anime/MF_CINE_ShadedDiffuseTint}.Result
   B=@n18.Result
   Alpha=NamedRerouteUsage()
  )
  n20 = LinearInterpolate(A=LinearInterpolate(A=Multiply(A=@n9, B=@n19), B=@n19, Alpha=NamedRerouteUsage()), B=@n19, Alpha=NamedRerouteUsage())
  n21 = ScalarParameter() {defaultValue=0.8, group=Toon_Shading_PBR_Override, parameterName=LitColor_PBR_CompensationScale, sliderMax=1, sortPriority=30}
  n22 = LinearInterpolate(
   A=MaterialFunctionCall!(
    In=MaterialFunctionCall!(In=@n20) {materialFunction=&/MF_TODCustomColors}.Result
    LUT
    LUT Multiplier
   ) {materialFunction=&/MF_ApplyLUT}.Result
   B=Multiply(A=@n21, B=MaterialFunctionCall(In=@n20) {materialFunction=&MF_ApplyLUT}.Result) {constB=0.5}
   Alpha=NamedRerouteUsage()
  )
  n23 = VectorParameter() {group=Toon_Shading_RimHighlighBacklit, parameterName=RimHighlight_Tint, sortPriority=20, defaultValue=<rgba 1 1 1 1>}
  n24 = ScalarParameter() {defaultValue=5, group=Toon_Shading_RimHighlighBacklit, parameterName=RimHighlight_Strength, sliderMax=10, sortPriority=16}
  n25 = Clamp(None=@n24) {maxDefault=5}
  n26 = Multiply(
   A=StaticSwitchParameter(True=@n23.RGB, False=@n22) {group=Toon_Shading_RimHighlighBacklit, parameterName=Use Rim Highlight Tint, sortPriority=0}
   B=MaterialFunctionCall(Value_Morning=@n24, Value_Day=@n24, Value_Evening=@n24, Value_Night=Multiply(A=@n25) {constB=0.5}) {materialFunction=&Scalar_Time_of_Day}.Result
  )
  n27 = NamedRerouteUsage()
  n28 = ScalarParameter() {defaultValue=5, group=Toon_Shading_RimHighlighBacklit, parameterName=RimHighlight_Strength, sliderMax=10, sortPriority=16}
  n29 = Multiply(A=@n18.Result, B=@n28)
  n30 = NamedRerouteUsage()
  n31 = StaticSwitchParameter(
   True=LinearInterpolate(
    A=@n22
    B=LinearInterpolate(A=@n22, B=@n26, Alpha=@n27)
    Alpha=ScalarParameter() {defaultValue=1, group=Toon_Shading_RimHighlighBacklit, parameterName=UseRimHighlight, sliderMax=1, sortPriority=15}
   )
   False=LinearInterpolate(
    A=@n20
    B=LinearInterpolate(A=@n20, B=@n29, Alpha=@n30)
    Alpha=ScalarParameter() {defaultValue=1, group=Toon_Shading_RimHighlighBacklit, parameterName=UseRimHighlight, sliderMax=1, sortPriority=15}
   )
  ) {group=Toon_Shading, parameterName=Use Apply LUT, sortPriority=0}
  n32 = NamedRerouteUsage()
  n33 = VectorParameter() {group=Toon_Shading_Hair, parameterName=SpecularColor, defaultValue=<rgba 1 1 1 1>}
  n34 = Multiply(
   A=Multiply(A=Multiply(A=@n32, B=@n33.RGB), B=ScalarParameter() {defaultValue=10, group=Toon_Shading_Hair, parameterName=SpecularStrength, sliderMax=20})
   B=LinearInterpolate(Alpha=NamedRerouteUsage()) {constA=1, constB=0.5}
  ) {constB=0.25}
  n35 = NamedRerouteUsage()
  n36 = ScalarParameter() {group=Toon_Shading_Hair, parameterName=IgnoreZforSpecularHighlight, sliderMax=1}
  n37 = StaticSwitchParameter(
   True=LinearInterpolate(A=ScalarParameter() {group=Toon_Shading_Hair, parameterName=IgnoreZforSpecularHighlight2, sliderMax=1}, B=@n36, Alpha=NamedRerouteUsage())
   False=@n36
  ) {
   desc=Use 2 IgnoreZforSpecularHighlight values based on preskin position.
   group=Toon_Shading_Hair
   parameterName=UseIgnoreZforSpecularHighlightBlend
   sortPriority=0
  }
  n38 = AppendVector(A=Constant2Vector() {g=1, r=1}, B=Saturate(None=OneMinus(None=@n37)))
  n39 = Multiply(A=NamedRerouteUsage(), B=@n38)
  n40 = Multiply(A=MaterialFunctionCall() {materialFunction=&CameraDirectionVector}.Result) {constB=-1}
  n41 = NamedRerouteUsage()
  n42 = Multiply(
   A=Add(
    A=CameraVectorWS()
    B=StaticSwitchParameter(True=@n40, False=@n41) {group=Toon_Shading_Hair, parameterName=UseCameraDirectionForSpecularHighlights, sortPriority=0}
   )
   B=@n38
  )
  n43 = Multiply(A=@n38, B=PixelNormalWS())
  n44 = StaticSwitchParameter(
   True=DotProduct(A=ReflectionVectorWS(), B=Normalize(VectorInput=@n39)) {desc=Phong}
   False=DotProduct(A=Normalize(VectorInput=@n42), B=Normalize(VectorInput=@n43)) {desc=Blinn}
  ) {desc=Switches between Blinn and Phong Highlight., group=Toon_Shading_Hair, parameterName=UsePhongHighlight, sortPriority=0}
  n45 = MaterialFunctionCall(
   In=Power(Base=Saturate(None=@n44), Exp=ScalarParameter() {defaultValue=50, group=Toon_Shading_Hair, parameterName=SpecPower, sliderMax=32})
   Contrast=ScalarParameter() {defaultValue=4, group=Toon_Shading_Hair, parameterName=Sharpness, sliderMax=8}
  ) {materialFunction=&CheapContrast}
  n46 = ScalarParameter() {
   defaultValue=0.2
   desc=Instensity of the specular highlight in the shaded area.
   group=Toon_Shading_Hair
   parameterName=ShadedSpecularHighlightIntensity
   sliderMax=1
  }
  n47 = NamedRerouteUsage()
  n48 = PreSkinnedPosition()
  n49 = NamedRerouteDeclaration(None=Divide(A=VertexInterpolator(VS=@n48).PS, B=PreSkinnedLocalBounds().Max)) {name=PreskinLocPosNormalized, nodeColor=<rgba 1 0.9609 0 1>}
  n50 = StaticSwitchParameter(
   True=TextureCoordinate() {coordinateIndex=1}
   False=StaticSwitchParameter(True=ComponentMask(None=@n49) {g=true, r=true}, False=TextureCoordinate()) {group=Toon_Shading_Hair, parameterName=UsePreskinnedPositionForSheen, sortPriority=0}
  ) {group=Toon_Shading_Hair, parameterName=UseUV2ForSheen, sortPriority=0}
  n51 = AppendVector(
   A=ScalarParameter() {defaultValue=7, group=Toon_Shading_Hair, parameterName=HairNoise_X_Tiling}
   B=ScalarParameter() {defaultValue=1, group=Toon_Shading_Hair, parameterName=HairNoise_Y_Tiling}
  )
  n52 = Multiply(
   A=TextureSample(UVs=Multiply(A=@n50, B=@n51)) {samplerSource=SSM_Wrap_WorldGroupSettings, texture=&T_NoisePack_01}.G
   B=ScalarParameter() {defaultValue=0.1, group=Toon_Shading_Hair, parameterName=HairNoiseStrength}
  )
  n53 = ScalarParameter() {defaultValue=2, group=Toon_Shading_Hair, parameterName=SheenBands}
  n54 = ScalarParameter() {defaultValue=0.2, group=Toon_Shading_Hair, parameterName=SheenBandMinThresh}
  n55 = ScalarParameter() {defaultValue=0.3, group=Toon_Shading_Hair, parameterName=SheenBandMaxThresh}
  n56 = Normalize(VectorInput=Add(
    A=MaterialFunctionCall() {materialFunction=&CameraDirectionVector}.Result
    B=NamedRerouteUsage()
   ))
  n57 = Constant3Vector() {constant=<rgba 0 0 1 0>}
  n58 = Multiply(
   A=OneMinus(None=DotProduct(A=@n56, B=@n57))
   B=ScalarParameter() {defaultValue=1, group=Toon_Shading_Hair, parameterName=SheenMovementThreshold}
  )
  n59 = Frac(None=Add(A=Multiply(A=@n50, B=@n53) {constB=2}, B=LinearInterpolate(A=@n54, B=@n55, Alpha=@n58)))
  n60 = CurveAtlasRowParameter(CurveTime=ChannelMaskParameter(None=Add(A=@n52, B=@n59)) {group=Toon_Shading_Hair, parameterName=SheenDirectionChannel, sortPriority=10}.R) {
   atlas=/Game/Characters/Shared/CurveAtlas/AthenaCharacterCurveAtlas
   curve=&/AnisoTropicHighLightCurve
   defaultValue=15
   group=Toon_Shading_Hair
   parameterName=AnisotropicSpecCurve
   sortPriority=5
  }
  n61 = StaticSwitchParameter(True=TextureCoordinate() {coordinateIndex=1}, False=TextureCoordinate()) {group=Toon_Shading_Hair, parameterName=UseUV2ForSheen, sortPriority=0}
  n62 = ComponentMask(None=@n61) {g=true}
  n63 = ScalarParameter() {defaultValue=128, group=Toon_Shading_Hair, parameterName=AnisoTropix_X_Tiling, sliderMax=256}
  n64 = ScalarParameter() {group=Toon_Shading_Hair, parameterName=AnisoTropix_Y_Tiling, sliderMax=256}
  n65 = Ceil(None=Multiply(A=@n61, B=AppendVector(A=@n63, B=@n64)))
  n66 = Add(
   A=ConstantBiasScale(None=Sine(None=@n62))
   B=ComponentMask(None=MaterialFunctionCall(Vec2=@n65) {materialFunction=&MF_Hash}.Vec2) {r=true}
  )
  n67 = Multiply(
   A=ConstantBiasScale(None=Sine(None=@n66))
   B=ScalarParameter() {defaultValue=0.1, group=Toon_Shading_Hair, parameterName=HairStrandSpread, sliderMax=1}
  ) {constB=0.5}
  n68 = Power(
   Base=Saturate(None=Subtract(A=@n60.G, B=@n67))
   Exp=ScalarParameter() {defaultValue=16, group=Toon_Shading_Hair, parameterName=HairStrandPower, sliderMax=32}
  )
  n69 = ScalarParameter() {defaultValue=1, group=Toon_Shading_Hair, parameterName=HairStrandSharpness, sliderMax=12}
  n70 = Add(
   A=Saturate(None=@n60.R)
   B=Multiply(
    A=MaterialFunctionCall(In=@n68, Contrast=@n69) {materialFunction=&CheapContrast}.Result
    B=ScalarParameter() {defaultValue=0.5, group=Toon_Shading_Hair, parameterName=HairStrandOpacity, sliderMax=1}
   )
  )
  n71 = MaterialFunctionCall(
   Alpha=NamedRerouteUsage()
   Min=ScalarParameter() {group=Toon_Shading_Hair, parameterName=HairSpecThreshold, sliderMax=1}
   Max=Constant() {r=1}
  ) {materialFunction=&SmoothStep}
  n72 = StaticSwitchParameter(
   True=Multiply(A=Multiply(A=@n35, B=@n45.Result), B=LinearInterpolate(A=@n46, Alpha=@n47))
   False=LinearInterpolate(
    A=@n45.Result
    B=LinearInterpolate(B=@n70, Alpha=@n71.Result)
    Alpha=ScalarParameter() {group=Toon_Shading_Hair, parameterName=UseAnisotropy, sliderMax=1, sortPriority=15}
   )
  ) {
   desc=Disables the Anisotropy setup and used a Fresnel to reveal the ssc B mask. adjust the specular color.
   group=Toon_Shading_Hair
   parameterName=UseSSCBlueasHighlight
   sortPriority=0
  }
  n73 = NamedRerouteDeclaration(None=@n72) {name=Painter Specular Highlight, nodeColor=<rgba 0 0.9364 1 1>}
  n74 = LinearInterpolate(
   A=@n31
   B=StaticSwitchParameter(True=LinearInterpolate(A=@n31, B=@n34, Alpha=@n73), False=@n31) {group=Toon_Shading, parameterName=UseSpecularHighlight, sortPriority=0}
   Alpha=NamedRerouteUsage()
  )
  n75 = Multiply(
   A=LinearInterpolate(
    A=NamedRerouteUsage()
    B=StaticSwitchParameter(True=@n8, False=@n74) {group=Toon_Shading, parameterName=UseFakePBR, sortPriority=0}
    Alpha=NamedRerouteUsage()
   )
   B=ScalarParameter() {defaultValue=1, group=Toon_Shading, parameterName=GlobalBrightness, sliderMax=2}
  )
  n76 = MaterialFunctionCall!(
   Emissive In=LinearInterpolate(
    A=MaterialFunctionCall!(Color=@n75) {materialFunction=&/MF_TODColorAdjustment}.Result
    B=@n75
    Alpha=NamedRerouteUsage()
   )
   LightingExtrasColorData=NamedRerouteUsage()
   CINE_LightSources=NamedRerouteUsage()
   ShadingCurve=NamedRerouteUsage()
  ) {materialFunction=/Game/Packages/Fortress_SharedMaterials/Cinematics/Anime/MF_CINE_FinalColorBlend}
  n77 = ChannelMaskParameter(None=NamedRerouteUsage()) {group=Toon_Shading_SSC, maskChannel=Green, parameterName=SSC_OverlapLineValueChannel, sortPriority=10, defaultValue=<rgba 0 1 0 0>}
  n78 = ScalarParameter() {defaultValue=0.001, group=Toon_Shading, parameterName=Subsurface Mask Value, sliderMax=1}
  n79 = SetMaterialAttributes!(
   MaterialAttributes=NamedRerouteUsage()
   BaseColor=Constant3Vector() {desc="dont touch!"}
   Metallic=ComponentMask(None=@n1) {desc=AO / Threshold, r=true}
   Specular=Constant() {desc="dont touch!"}
   EmissiveColor=Add(A=NamedRerouteUsage(), B=NamedRerouteDeclaration(None=@n76.Result) {name=FinalColor, nodeColor=<rgba 1 0.4967 0 1>})
   Roughness=LinearInterpolate(A=@n77.R, B=OneMinus(None=@n77.R), Alpha=ScalarParameter() {group=Toon_Shading_SSC, parameterName=Invert Roughness Pass, sliderMax=1})
   AmbientOcclusion=NamedRerouteUsage()
   SubsurfaceColor=MaterialFunctionCall(Y=@n78, Z=@n78) {materialFunction=&MakeFloat3}.Result
   CustomData0=Constant() {desc=Opacity, r=1}
   ShadingModel=ShadingModel() {shadingModel=MSM_Cloth}
  ) {attributeSetTypes [
    #1
    #2
    #3
    #4
    #5
    #6
    #7
    #8
    #9
   ]}
  n80 = SetMaterialAttributes!(
   MaterialAttributes=@n79
   BaseColor=@n3
   EmissiveColor=Add(A=Multiply(A=@n3, B=@n4), B=Multiply(A=@n5, B=@n6))
   AmbientOcclusion=Constant() {r=1}
   Metallic=@n7
   Specular=@n7
  ) {attributeSetTypes [
    #1
    #4
    #6
    #2
    #3
   ]}
  n81 = SetMaterialAttributes!(
   MaterialAttributes=@n80
   Specular=ScalarParameter() {group=Toon_Shading_PBR_Override, parameterName=Specular_Override, sliderMax=1, sortPriority=21}
   Roughness=ScalarParameter() {defaultValue=1, group=Toon_Shading_PBR_Override, parameterName=Roughness_Override, sliderMax=1, sortPriority=22}
   Metallic=ScalarParameter() {group=Toon_Shading_PBR_Override, parameterName=Metallic_Override, sliderMax=1, sortPriority=23}
  ) {attributeSetTypes [#3, #5, #2]}
  n82 = Saturate(None=ScalarParameter() {group=Toon_Shading_PBR_Override, parameterName=PBR_Shading, sliderMax=1, sortPriority=10})
  FunctionOutput(None=BlendMaterialAttributes(
    A=@n79
    B=StaticSwitchParameter(True=@n81, False=@n80) {group=Toon_Shading_PBR_Override, parameterName=UseOverrideValues, sortPriority=0}
    Alpha=NamedRerouteDeclaration(None=@n82) {name=PBR_Switch, nodeColor=<rgba 0 1 0.8566 1>}
   )) {sortPriority=10}
  FunctionOutput(None=NamedRerouteUsage()) {outputName=SSC, sortPriority=40}
  FunctionOutput(None=NamedRerouteUsage()) {outputName=InkLine Mask, sortPriority=30}
  FunctionOutput(None=NamedRerouteUsage()) {outputName=PBR_Switch, sortPriority=60}
  FunctionOutput(None=NamedRerouteUsage()) {outputName=FX, sortPriority=50}
  FunctionOutput(None=NamedRerouteUsage()) {description=R:Shading Curve G:Specular Highlights B:Empty, outputName=Shading, sortPriority=20}
  n83 = TextureSampleParameter2D(UVs=NamedRerouteUsage()) {
   group=Toon_Shading
   parameterName=SSC_Texture
   samplerSource=SSM_Wrap_WorldGroupSettings
   samplerType=SAMPLERTYPE_Masks
   sortPriority=1
   texture=&T_BlackMask
  }
  n84 = MaterialFunctionCall(Float4=@n83.RGBA) {materialFunction=&BreakOutFloat4Components}
  n85 = StaticSwitchParameter(
   True=MaterialFunctionCall(
    X=@n84.R
    Y=@n84.R
    Z=@n84.R
    A=FunctionInput() {bUsePreviewValueAsDefault=true, inputName=SSC_AlphaOverride, inputType=FunctionInput_Scalar, sortPriority=1}
   ) {materialFunction=&MakeFloat4}.Result
   False=@n83.RGBA
  ) {group=Toon_Shading_SSC, parameterName=SSC_OverrideAlphaChannel, sortPriority=0}
  n86 = TextureSampleParameter2D(UVs=NamedRerouteUsage()) {
   group=Toon_Shading
   parameterName=SSC_Texture_Far
   samplerSource=SSM_Wrap_WorldGroupSettings
   samplerType=SAMPLERTYPE_Masks
   sortPriority=1
   texture=&T_BlackMask
  }
  n87 = ScalarParameter() {defaultValue=100, group=Toon_Shading_SSC, parameterName=SSC_FarDistance, sliderMax=1000}
  n88 = Subtract(A=Distance(A=WorldPosition().XYZ, B=CameraPositionWS()), B=@n87)
  n89 = Subtract(A=ScalarParameter() {defaultValue=500, group=Toon_Shading_SSC, parameterName=SSC_NearDistance, sliderMax=1000}, B=@n87)
  n90 = Divide(
   A=ComponentMask(None=ViewProperty() {property=MEVP_BufferSize}.Property) {r=true}
   B=ScalarParameter() {defaultValue=3840, group=Toon_Shading_SSC, parameterName=Reference Screen Size, sliderMax=3840}
  )
  n91 = Saturate(None=Divide(A=Divide(A=@n88, B=@n89), B=Power(Base=@n90) {constExponent=1.2}))
  n92 = TextureSampleParameter2D(UVs=TextureCoordinate() {coordinateIndex=1}) {
   desc=Only overrides green channel in original SSC
   group=Toon_Shading
   parameterName=SSC_Texture_2nd_UV
   samplerType=SAMPLERTYPE_Masks
   sortPriority=1
   texture=&T_BlackMask
  }
  n93 = ChannelMaskParameter(None=@n92.RGBA) {group=Toon_Shading_SSC, maskChannel=Green, parameterName=SSC_Texture_2nd_UV_Channel, sortPriority=10, defaultValue=<rgba 0 1 0 0>}
  n94 = MaterialFunctionCall(X=@n83.R, Y=@n93.R, Z=@n83.R, A=@n83.R) {materialFunction=&MakeFloat4}
  NamedRerouteDeclaration(None=StaticSwitchParameter(
    True=LinearInterpolate(A=@n85, B=@n86.RGBA, Alpha=@n91)
    False=StaticSwitchParameter(True=@n94.Result, False=@n85) {desc=Doesn't work with Enable Far SSC, group=Toon_Shading_SSC, parameterName=Enable 2nd UV Channel SSC, sortPriority=0}
   ) {group=Toon_Shading, parameterName=Enable Far SSC, sortPriority=0}) {name=SSC Map, nodeColor=<rgba 1 0 0.4453 1>}
  n95 = NamedRerouteUsage()
  n96 = Sign(None=ComponentMask(None=Transform(None=@n95) {transformSourceType=TRANSFORMSOURCE_World, transformType=TRANSFORM_Local}) {r=true})
  n97 = NamedRerouteUsage()
  n98 = TextureObjectParameter() {
   desc=ShadowGradient Texture
   group=Toon_Shading
   parameterName=SG_Texture
   samplerType=SAMPLERTYPE_LinearGrayscale
   sortPriority=1
   texture=/Game/Global/Textures/Default/Blanks/T_BlackLinearGrayscale
  }
  n99 = TextureSample(UVs=Multiply(A=@n97, B=Constant2Vector() {g=1, r=-1}), Tex=@n98) {samplerType=SAMPLERTYPE_LinearGrayscale, texture=&T_BlackLinearGrayscale}
  n100 = Multiply(A=If(
    A=Saturate(None=@n96)
    "A > B"=TextureSample(UVs=@n97, Tex=@n98) {samplerType=SAMPLERTYPE_LinearGrayscale, texture=&T_BlackLinearGrayscale}.R
    "A < B"=OneMinus(None=@n99.R)
   ) {constB=0.5}) {constB=2}
  n101 = ScalarParameter() {defaultValue=1, group=Toon_Shading, parameterName=SG_LightAngleSideWeight, sliderMax=2}
  n102 = VertexColor()
  n103 = MaterialFunctionCall(Float3=Transform(None=PixelNormalWS()) {transformSourceType=TRANSFORMSOURCE_World, transformType=TRANSFORM_Local}) {materialFunction=&BreakOutFloat3Components}
  n104 = MaterialFunctionCall(X=Multiply(A=Add(A=@n100) {constB=-1}, B=Multiply(A=@n101, B=@n102.A)), Y=@n103.G, Z=@n103.G) {materialFunction=&MakeFloat3}
  n105 = TextureSampleParameter2D(UVs=NamedRerouteUsage()) {
   desc=ShadowGradient Texture
   group=Toon_Shading
   parameterName=SG_Texture
   samplerType=SAMPLERTYPE_LinearGrayscale
   sortPriority=1
   texture=&T_BlackLinearGrayscale
  }
  n106 = Multiply(A=@n105.R) {constB=2}
  n107 = MaterialFunctionCall(Float3=Transform(None=PixelNormalWS()) {transformSourceType=TRANSFORMSOURCE_World, transformType=TRANSFORM_Local}) {materialFunction=&BreakOutFloat3Components}
  n108 = Normalize(VectorInput=MaterialFunctionCall(X=Subtract(A=@n106), Y=@n107.G, Z=@n107.G) {materialFunction=&MakeFloat3}.Result)
  n109 = ScalarParameter() {defaultValue=0.5, desc=ShadowGradient Intensity Blending, group=Toon_Shading, parameterName=SG_Blend, sliderMax=1}
  n110 = VertexColor()
  n111 = WorldPosition()
  n112 = Add(
   A=VectorParameter() {group=Toon_Shading_SphericalNormal, parameterName=SocketOffset, defaultValue=<rgba 0 0 0 1>}.RGB
   B=VectorParameter() {group=Toon_Shading_SphericalNormal, parameterName=HeadSocketLocation, defaultValue=<rgba 0 0 0 1>}.RGB
  )
  n113 = Multiply(
   A=Normalize(VectorInput=Subtract(A=@n111.XYZ, B=@n112))
   B=ScalarParameter() {defaultValue=1, group=Toon_Shading_SphericalNormal, parameterName=SphericalNormalScale, sliderMax=1}
  )
  n114 = ScalarParameter() {defaultValue=0.5, group=Toon_Shading_SphericalNormal, parameterName=SphericalNormal_Blend_Amt, sliderMax=1}
  n115 = Subtract(A=PreSkinnedPosition(), B=VectorParameter() {group=Toon_Shading, parameterName=NormalSphereLocation, defaultValue=<rgba 0 0 162 1>}.RGB)
  n116 = Constant3Vector() {constant=<rgba 1 1 0 0>}
  n117 = Normalize(VectorInput=LinearInterpolate(
    A=@n115
    B=Multiply(A=@n115, B=@n116)
    Alpha=ScalarParameter() {group=Toon_Shading, parameterName=Sphere to Cylindrical Normals, sliderMax=1}
   ))
  n118 = LinearInterpolate(
   A=PixelNormalWS()
   B=Transform(None=VertexInterpolator(VS=@n117).PS) {transformSourceType=TRANSFORMSOURCE_Local}
   Alpha=ScalarParameter() {group=Toon_Shading, parameterName=NormalSphereBlend, sliderMax=1}
  ) {constAlpha=0}
  n119 = Transform(None=Constant3Vector() {constant=<rgba 0 0 1 0>})
  n120 = ScalarParameter() {group=Toon_Shading, parameterName=FlatNormalBlend, sliderMax=1}
  n121 = ScalarParameter() {group=Toon_Shading, parameterName=FakeNormalBlend_Amt, sliderMax=1}
  n122 = StaticSwitchParameter(True=Multiply(A=@n121, B=VertexColor().A), False=@n121) {
   desc=Fades out the painted face normals by the vertex colors alpha channel.
   group=Toon_Shading
   parameterName=UseBlendFakeNormalsByVCAlpha
   sortPriority=0
  }
  n123 = Subtract(A=Multiply(A=VertexColor()) {constB=2})
  n124 = Normalize(VectorInput=LinearInterpolate(A=LinearInterpolate(A=@n118, B=@n119, Alpha=@n120), B=Transform(None=@n123) {transformSourceType=TRANSFORMSOURCE_Local}, Alpha=@n122))
  n125 = StaticSwitchParameter(True=Normalize(VectorInput=LinearInterpolate(A=@n124, B=@n113, Alpha=@n114)), False=@n124) {group=Toon_Shading, parameterName=UseSphericalNormalBlend, sortPriority=0}
  n126 = StaticSwitchParameter(
   True=Transform(None=Normalize(VectorInput=@n104.Result)) {transformSourceType=TRANSFORMSOURCE_Local}
   False=LinearInterpolate(A=@n125, B=Transform(None=@n108) {transformSourceType=TRANSFORMSOURCE_Local}, Alpha=Multiply(A=@n109, B=@n110.A))
  ) {
   desc="Activate to use the ShadowGradient Map for the painted shading control.\r\nwith Flipping Tech"
   group=Toon_Shading
   parameterName=UseSGMapFlippingTech
   sortPriority=0
  }
  n127 = StaticSwitchParameter(True=@n126, False=@n125) {desc=Activate to use the ShadowGradient Map for the painted shading control., group=Toon_Shading, parameterName=UseSGMap, sortPriority=0}
  NamedRerouteDeclaration(None=@n127) {name=Adjusted Normals, nodeColor=<rgba 0.5 0.5 1 1>}
  n128 = TextureSampleParameter2D(UVs=NamedRerouteUsage()) {
   desc="R = SpecMask G = Metallic B = Roughness"
   group=Specular
   parameterName=SpecularMasks
   samplerType=SAMPLERTYPE_Masks
   texture=&T_Fortnite_Default_S
  }
  NamedRerouteDeclaration(None=@n128.R) {name=Specular Input, nodeColor=<rgba 1 1 0 1>}
  NamedRerouteDeclaration(None=@n128.G) {name=Metallic Input, nodeColor=<rgba 1 1 0 1>}
  n129 = TextureSampleParameter2D(UVs=NamedRerouteUsage()) {
   group=Skin FX
   parameterName=SkinFX_Mask
   samplerType=SAMPLERTYPE_Masks
   sortPriority=0
   texture=&T_BlackMask
  }
  NamedRerouteDeclaration(None=@n129.RGBA) {name=SkinFX, nodeColor=<rgba 1 0.3984 0 1>}
  n130 = GetMaterialAttributes(None=FunctionInput() {inputName=Material, inputType=FunctionInput_MaterialAttributes}) {attributeGetTypes [#4]}
  NamedRerouteDeclaration(None=@n130.EmissiveColor) {name=EmissiveColor, nodeColor=<rgba 1 1 0 1>}
  n131 = Constant3Vector() {constant=<rgba 0 0 1 0>}
  n132 = ChannelMaskParameter(None=NamedRerouteUsage()) {
   desc="Select in which channel of the FX map the mask camera light vector will be stored "
   group=Toon_Shading
   parameterName=CameraLightVector_FXChannel
   sortPriority=10
  }
  MaterialExpressionReroute_12 = Reroute(None=@n2.RGB)
  n133 = Normalize(VectorInput=@MaterialExpressionReroute_12.LightVector)
  n134 = ComponentMask(None=@n133) {r=true}
  n135 = Constant3Vector() {constant=<rgba 0 1 0 0>}
  n136 = Normalize(VectorInput=WorldPosition() {worldPositionShaderOffset=WPT_CameraRelative}.XYZ)
  n137 = Normalize(VectorInput=CrossProduct(A=Transform(None=@n135) {transformSourceType=TRANSFORMSOURCE_View}, B=@n136))
  n138 = ComponentMask(None=@n133) {g=true}
  n139 = CrossProduct(A=@n136, B=@n137)
  n140 = Add(A=Add(A=Multiply(A=@n134, B=@n137), B=Multiply(A=@n138, B=@n139)), B=Multiply(A=ComponentMask(None=@n133) {b=true}, B=@n136))
  MaterialExpressionReroute_14 = Reroute(None=@n140)
  n141 = ComponentMask(None=Custom() {code="return View.DirectionalLightColor;", description=DirectionalLightColor, outputType=CMOT_Float4, inputs []}) {b=true, g=true, r=true}
  n142 = Custom() {code="return View.DirectionalLightDirection;", description=DirectionalLightDirection, inputs []}
  n143 = CollectionParameter() {
   collection=&FortniteMaterialParameters
   parameterName=SunlightDirectionalVector
  }
  n144 = ComponentMask(None=@n143) {b=true, g=true, r=true}
  n145 = If(A=DotProduct(A=@n142, B=@n142), "A > B"=@n142, "A == B"=@n144, "A < B"=@n144) {desc=check if DirectionalLight exists}
  n146 = StaticSwitchParameter(
   True=@MaterialExpressionReroute_14.Result
   False=MaterialFunctionCall!(In=If(A=@n141, "A > B"=@n145, "A == B"=@n144, "A < B"=@n144) {desc=check if DirectionalLight color is Black}) {materialFunction=/Game/Packages/Fortress_SharedMaterials/Cinematics/Anime/MF_CINE_AnimeLightDirOverride}.Result
  ) {
   desc="Use a Screen / View alight light vector. \r\nOverrides the UseDirectionalLightFunction!"
   group=Toon_Shading
   parameterName=UseScreenSpaceLightDirection
   sortPriority=0
  }
  n147 = MaterialFunctionCall(Float3=@n146) {materialFunction=&BreakOutFloat3Components}
  n148 = ScalarParameter() {
   desc="0.0 = the light vector is locked on the XY plane.\r\n1.0 = the actual light vector is used  \r\n(for Faces use a blend value of 0,0)"
   group=Toon_Shading
   parameterName=LockLightZ_Blend
   sliderMax=1
  }
  n149 = MaterialFunctionCall(X=@n147.R, Y=@n147.R, Z=Multiply(A=@n147.B, B=@n148)) {materialFunction=&MakeFloat3}
  n150 = StaticSwitchParameter(True=LinearInterpolate(A=@n146, B=Normalize(VectorInput=@n149.Result), Alpha=VertexColor().A), False=@n146) {
   desc="Lock the Z axis of the Light Vector\r\nis driven by the VertexColor A channel (0 = Regular LightVector, 1= Locked LIghtVector)"
   group=Toon_Shading
   parameterName=UseLockLightZVector
   sortPriority=0
  }
  n151 = ComponentMask(None=@n150) {b=true, g=true, r=true}
  n152 = StaticSwitchParameter(
   True=LinearInterpolate(A=@n151, B=Transform(None=@n131) {desc=Camera Forward Vector, transformSourceType=TRANSFORMSOURCE_Camera}, Alpha=@n132.R)
   False=@n151
  ) {
   desc="Use the Camera vector instead of the light vector for an reaz which is defined by the FX map. Select the FX channel with this parameter \"CameraLightVector_FXChannel\"."
   group=Toon_Shading
   parameterName=UseCameraLightVector
   sortPriority=0
  }
  NamedRerouteDeclaration(None=@n152) {name=LightVector, nodeColor=<rgba 0 1 0.8634 1>}
  NamedRerouteDeclaration(None=@n130.MaterialAttributes) {name=MA, nodeColor=<rgba 1 1 0 1>}
  n153 = NamedRerouteUsage()
  n154 = TextureSampleParameter2D(UVs=@n153) {
   group=Toon_Shading
   parameterName=DistanceField_InkLines
   samplerSource=SSM_Wrap_WorldGroupSettings
   samplerType=SAMPLERTYPE_Alpha
   sortPriority=1
   texture=/Game/Global/Textures/Default/Blanks/T_BlackAlpha
  }
  n155 = TextureSampleParameter2D(UVs=@n153) {
   group=Toon_Shading
   parameterName=DistanceField_InkLines_Mobile
   samplerSource=SSM_Wrap_WorldGroupSettings
   samplerType=SAMPLERTYPE_Alpha
   sortPriority=1
   texture=&T_BlackAlpha
  }
  n156 = StaticSwitchParameter(
   True=QualitySwitch(Default=@n154.RGBA, Low=@n155.RGBA, High=@n154.RGBA, Medium=@n155.RGBA, Epic=@n154.RGBA)
   False=StaticSwitchParameter(True=ShadingPathSwitch(Default=@n154.RGBA, Deferred=@n154.RGBA, Mobile=@n155.RGBA) {desc=Migrated from FeatureLevelSwitch}, False=@n154.RGBA) {group=Toon_Shading, parameterName=Enable Unique Mobile DF Inklines, sortPriority=0}
  ) {group=Toon_Shading, parameterName=Enable Unique Low Quality DF Inklines, sortPriority=0}
  n157 = MaterialFunctionCall!(Roughness input=StaticSwitchParameter(True=Power(Base=@n156) {constExponent=0.4545}, False=@n156) {group=Toon_Shading, parameterName=UseGammaCorrectedDFLines, sortPriority=0}) {materialFunction=&/MF_OrientationFade}
  n158 = FunctionInput() {
   bUsePreviewValueAsDefault=true
   desc=This mask only affects DistanceField_InkLines texture parameter
   inputName=DF InkLine Visibility Mask
   inputType=FunctionInput_Scalar
   sortPriority=10
   previewValue=<1 0 0 1>
  }
  n159 = Max(
   A=FunctionInput() {bUsePreviewValueAsDefault=true, inputName=DF Inkline Injection, inputType=FunctionInput_Scalar, sortPriority=11}
   B=Multiply(A=ChannelMaskParameter(None=@n157.Result) {group=Toon_Shading_DF_InkLines, parameterName=DistanceField_Channel, sortPriority=10}.R, B=@n158) {desc=Mask Inkline Here}
  ) {desc=Inject Inklines here}
  n160 = ScalarParameter() {defaultValue=4, group=Toon_Shading_DF_InkLines, parameterName=DF_BlurPow, sliderMax=16}
  n161 = OneMinus(None=Multiply(
    A=Power(Base=@n159, Exp=@n160) {constExponent=3}
    B=ScalarParameter() {defaultValue=1, group=Toon_Shading_DF_InkLines, parameterName=DF_BlurMult, sliderMax=2}
   ) {constB=0.5})
  n162 = Constant() {r=1}
  n163 = ScalarParameter() {defaultValue=0.4, group=Toon_Shading_DF_InkLines, parameterName=DF_LineThickness_Near, sliderMax=1}
  n164 = ScalarParameter() {defaultValue=0.4, group=Toon_Shading_DF_InkLines, parameterName=DF_LineThickness_Near_Mobile, sliderMax=1}
  n165 = ScalarParameter() {defaultValue=0.6, group=Toon_Shading_DF_InkLines, parameterName=DF_LineThickness_Far, sliderMax=1}
  n166 = ScalarParameter() {defaultValue=0.6, group=Toon_Shading_DF_InkLines, parameterName=DF_LineThickness_Far_Mobile, sliderMax=1}
  n167 = MaterialFunctionCall!(
   Fade Offset=ScalarParameter() {defaultValue=100, group=Toon_Shading_DF_InkLines, parameterName=DF_FarDistance, sliderMax=1000}
   Fade Distance=ScalarParameter() {defaultValue=500, group=Toon_Shading_DF_InkLines, parameterName=DF_NearDistance, sliderMax=1000}
  ) {materialFunction=/Game/Global/Materials/Character_Parent_2023/Anime_Parent_2023/MaterialFunctions/MFC_DistanceFade_FOV_compensation}
  n168 = StaticSwitchParameter(
   True=LinearInterpolate(A=ShadingPathSwitch(Default=@n163, Mobile=@n164), B=ShadingPathSwitch(Default=@n165, Mobile=@n166), Alpha=@n167.Result)
   False=LinearInterpolate(A=@n163, B=@n165, Alpha=@n167.Result)
  ) {group=Toon_Shading, parameterName=Enable Mobile DF Inklines Thickness, sortPriority=0}
  n169 = ScalarParameter() {defaultValue=0.8, group=Toon_Shading_DF_InkLines, parameterName=DF_LineHardness, sliderMax=1}
  n170 = OneMinus(None=Saturate(None=SphereMask(A=@n159, B=@n162, Radius=@n168, Hardness=@n169) {attenuationRadius=0.1}))
  NamedRerouteDeclaration(None=Saturate(None=Multiply(A=@n161, B=@n170))) {name=InkLine Mask, nodeColor=<rgba 0.1615 0.1615 0.1615 1>}
  n171 = TextureSampleParameter2D(UVs=NamedRerouteUsage()) {group=Toon_Shading, parameterName=InkLineColor_Texture, sortPriority=1, texture=&T_BlacksRGB}
  n172 = MaterialFunctionCall(In=@n171.RGB) {materialFunction=&MF_ApplyLUT}
  n173 = VectorParameter() {group=Toon_Shading_DF_InkLines, parameterName=InkLineColor, defaultValue=<rgba 0 0 0 1>}
  n174 = FunctionInput(Preview=@n173.RGB) {bUsePreviewValueAsDefault=true, inputName=DF Inkline Color Override, sortPriority=12}
  n175 = FunctionInput() {bUsePreviewValueAsDefault=true, inputName=DF Inkline Color Override Mask, inputType=FunctionInput_Scalar, sortPriority=13}
  NamedRerouteDeclaration(None=StaticSwitchParameter(
    True=StaticSwitchParameter(True=@n172.Result, False=@n171.RGB) {group=Toon_Shading, parameterName=Use Apply LUT, sortPriority=0}
    False=LinearInterpolate(A=@n173.RGB, B=@n174, Alpha=@n175)
   ) {group=Toon_Shading, parameterName=UseInkLineColorTexture, sortPriority=0}) {name=InkLine Color, nodeColor=<rgba 0.1615 0.1615 0.1615 1>}
  NamedRerouteDeclaration(None=FunctionInput(Preview=TextureCoordinate()) {bUsePreviewValueAsDefault=true, inputName=CustomUVs, inputType=FunctionInput_Vector2, sortPriority=15}) {name=CustomUVs, nodeColor=<rgba 1 0.0993 0 1>}
  NamedRerouteDeclaration(None=StaticSwitchParameter(True=@n22, False=@n20) {group=Toon_Shading, parameterName=Use Apply LUT, sortPriority=0}) {name=Base Color OUT, nodeColor=<rgba 0 1 0.5705 1>}
  n176 = NamedRerouteUsage()
  n177 = NamedRerouteUsage()
  NamedRerouteDeclaration(None=MaterialFunctionCall(X=NamedRerouteUsage(), Y=StaticSwitchParameter(True=@n176, False=@n177) {group=Toon_Shading, parameterName=UseFakePBR, sortPriority=0}) {materialFunction=&MakeFloat3}.Result) {name=Shading, nodeColor=<rgba 0 0.3906 1 1>}
  n178 = NamedRerouteUsage()
  n179 = Divide(
   A=Distance(
    A=ComponentMask(None=@n178) {b=true}
    B=ScalarParameter() {defaultValue=0.25, group=Toon_Shading_Hair, parameterName=IgnoreZforSpecularHighlight_BlendPosition}
   )
   B=ScalarParameter() {defaultValue=0.325, group=Toon_Shading_Hair, parameterName=IgnoreZforSpecularHighlight_BlendSpread}
  ) {constB=0.4}
  n180 = ScalarParameter() {defaultValue=4, group=Toon_Shading_Hair, parameterName=IgnoreZforSpecularHighlight_BlendContrast}
  NamedRerouteDeclaration(None=Saturate(None=MaterialFunctionCall(In=@n179, Contrast=@n180) {materialFunction=&CheapContrast}.Result)) {name=IgnoreZblend, nodeColor=<rgba 0 0.4141 1 1>}
  n181 = MaterialFunctionCall!(In=DotProduct(A=NamedRerouteUsage(), B=NamedRerouteUsage())) {materialFunction=/Game/Packages/Fortress_SharedMaterials/Cinematics/Anime/MF_CINE_AnimeLightingFeatures}
  NamedRerouteDeclaration(None=@n181.CINE_LightSources) {name=CINE_LightSources, nodeColor=<rgba 1 0.9609 0 1>}
  NamedRerouteDeclaration(None=@n181.CINE_LightColor) {name=CINE_LightColor, nodeColor=<rgba 1 0.6803 0 1>}
  n182 = NamedRerouteUsage()
  n183 = PixelNormalWS()
  n184 = Fresnel(
   ExponentIn=ScalarParameter() {defaultValue=4, group=Toon_Shading_RimHighlighBacklit, parameterName=BL_RimHighlightExponent1, sliderMax=8}
   Normal=PixelNormalWS()
  ) {baseReflectFraction=0}
  n185 = LinearInterpolate(
   A=ScalarParameter() {defaultValue=-4, group=Toon_Shading_RimHighlighBacklit, parameterName=BL_Rim_OutlineC1, sliderMax=16, sliderMin=-8}
   B=ScalarParameter() {defaultValue=12, group=Toon_Shading_RimHighlighBacklit, parameterName=BL_Rim_OutlineC2, sliderMax=16, sliderMin=-8}
   Alpha=Power(
    Base=OneMinus(None=@n184)
    Exp=ScalarParameter() {defaultValue=4, group=Toon_Shading_RimHighlighBacklit, parameterName=BL_Rim_OutlineExponent2, sliderMax=8}
   )
  )
  n186 = Multiply(A=Saturate(None=DotProduct(A=@n182, B=@n183)), B=OneMinus(None=Saturate(None=@n185)))
  n187 = NamedRerouteUsage()
  n188 = StaticSwitchParameter(
   True=Multiply(A=@n186, B=ChannelMaskParameter(None=@n187) {group=Toon_Shading_RimHighlighBacklit, parameterName=Rim_FXMask_Channel, sortPriority=10}.R)
   False=@n186
  ) {group=Toon_Shading_RimHighlighBacklit, parameterName=UseFXMaskforRimHighlight, sortPriority=0}
  n189 = VectorParameter() {
   desc=CameraVectortoReduceRimHighlight
   group=Toon_Shading_RimHighlighBacklit
   parameterName=RimHighlight_CustomTopVector
   defaultValue=<rgba 0 0 1 1>
  }
  n190 = ComponentMask(None=Transform(None=Normalize(VectorInput=@n189.RGB)) {transformSourceType=TRANSFORMSOURCE_Camera, transformType=TRANSFORM_Local}) {b=true}
  n191 = Power(Base=Saturate(None=Multiply(A=@n190) {constB=-1})) {constExponent=10}
  n192 = ScalarParameter() {defaultValue=1, group=Toon_Shading_RimHighlighBacklit, parameterName=RimHighlight_TopReducing, sliderMax=10}
  n193 = Saturate(None=Add(A=OneMinus(None=@n191), B=OneMinus(None=@n192)))
  NamedRerouteDeclaration(None=StaticSwitchParameter(True=Multiply(A=@n188, B=@n193), False=@n188) {group=Toon_Shading_RimHighlighBacklit, parameterName=UseReduceTopRimHighlight, sortPriority=0}) {name=Rim Highlight, nodeColor=<rgba 1 0 0.1172 1>}
  n194 = NamedRerouteUsage()
  n195 = MaterialFunctionCall(
   Alpha=Power(
    Base=ChannelMaskParameter(None=@n194) {group=Toon_Shading_SSC, parameterName=SSC_ThresholdChannel, sortPriority=10}.R
    Exp=ScalarParameter() {defaultValue=2, group=Toon_Shading, parameterName=ThresholdmaskPower, sliderMax=8}
   )
   Min=ScalarParameter() {group=Toon_Shading, parameterName=Shaded_Threshold_Min, sliderMax=1}
   Max=ScalarParameter() {defaultValue=1, group=Toon_Shading, parameterName=Shaded_Threshold_Max, sliderMax=1}
  ) {materialFunction=&SmoothStep}
  NamedRerouteDeclaration(None=Round(None=@n195.Result)) {name=SSC Throshold, nodeColor=<rgba 0.7973 1 0 1>}
  n196 = Saturate(None=@n181.CINE_ShadingCurve)
  NamedRerouteDeclaration(None=Round(None=@n196)) {name=Shading Dot, nodeColor=<rgba 0.7973 1 0 1>}
  n197 = Add(A=@n181.CINE_ShadingCurve, B=ScalarParameter() {group=Toon_Shading, parameterName=Lit Bias, sliderMax=1})
  n198 = ScalarParameter() {defaultValue=12, group=Toon_Shading_RimHighlighBacklit, parameterName=BackLitRim_OutlineC2, sliderMax=16, sliderMin=-8}
  n199 = Fresnel(
   ExponentIn=ScalarParameter() {defaultValue=2, group=Toon_Shading_RimHighlighBacklit, parameterName=BackLitExponent1, sliderMax=8}
   Normal=NamedRerouteUsage()
  ) {baseReflectFraction=0}
  n200 = LinearInterpolate(
   A=ScalarParameter() {defaultValue=-4, group=Toon_Shading_RimHighlighBacklit, parameterName=BackLitRim_OutlineC1, sliderMax=16}
   B=@n198
   Alpha=Power(
    Base=OneMinus(None=@n199)
    Exp=ScalarParameter() {defaultValue=4, group=Toon_Shading_RimHighlighBacklit, parameterName=BacklitOutlineExponent2, sliderMax=8}
   )
  )
  n201 = Add(A=OneMinus(None=Saturate(None=@n200)), B=@n196)
  n202 = Add(
   A=DotProduct(
    A=MaterialFunctionCall() {materialFunction=&CameraDirectionVector}.Result
    B=NamedRerouteUsage()
   )
   B=ScalarParameter() {defaultValue=0.25, group=Toon_Shading_RimHighlighBacklit, parameterName=BackLitBias, sliderMax=1}
  )
  n203 = ScalarParameter() {defaultValue=0.75, group=Toon_Shading_RimHighlighBacklit, parameterName=BackLitMaxBias, sliderMax=1}
  n204 = CurveAtlasRowParameter(CurveTime=LinearInterpolate(A=Multiply(A=@n197, B=@n195.Result), B=Multiply(A=@n201, B=@n195.Result), Alpha=Clamp(None=@n202, Max=@n203))) {
   atlas=&AthenaCharacterCurveAtlas
   curve=&/TwoToneShadingCurve
   defaultValue=14
   group=Toon_Shading
   parameterName=ShadingCurve
   sortPriority=5
  }
  NamedRerouteDeclaration(None=LinearInterpolate(A=ChannelMaskParameter(None=@n204) {group=Toon_Shading, parameterName=ShadingCurveChannel, sortPriority=10}.R, Alpha=NamedRerouteUsage())) {name=Shading Curve, nodeColor=<rgba 0.7973 1 0 1>}
  n205 = VectorParameter() {group=Toon_Shading_Fake PBR, parameterName=Specular Tint, defaultValue=<rgba 1 1 1 1>}
  n206 = Multiply(A=@n205.RGB, B=NamedRerouteUsage()) {constA=1}
  n207 = NamedRerouteUsage()
  n208 = ScalarParameter() {defaultValue=1, group=Toon_Shading_Fake PBR, parameterName=Specular Strength, sliderMax=1}
  n209 = LinearInterpolate(Alpha=NamedRerouteUsage()) {constA=1, constB=0.25}
  n210 = ScalarParameter() {defaultValue=0.0005, group=Toon_Shading_Fake PBR, parameterName=Roughness Min, sliderMax=1}
  n211 = ScalarParameter() {defaultValue=1, group=Toon_Shading_Fake PBR, parameterName=Roughness Max, sliderMax=1}
  n212 = Saturate(None=Power(Base=NamedRerouteUsage()) {constExponent=3})
  n213 = NamedRerouteUsage()
  n214 = MaterialFunctionCall() {materialFunction=&CameraDirectionVector}
  n215 = Add(
   A=CameraVectorWS()
   B=LinearInterpolate(
    A=ComponentMask(None=@n213) {b=true, g=true, r=true}
    B=Multiply(A=@n214.Result) {constB=-1}
    Alpha=ScalarParameter() {group=Toon_Shading_Fake PBR, parameterName=CameraDirectionForSpecularHighlights, sliderMax=1}
   )
  )
  n216 = StaticSwitchParameter(
   True=ChannelMaskParameter(None=NamedRerouteUsage()) {group=Toon_Shading_Fake PBR, parameterName=Ignore Z SkinFX Channel, sortPriority=10}.R
   False=ScalarParameter() {group=Toon_Shading_Fake PBR, parameterName=Ignore Z, sliderMax=1}
  ) {group=Toon_Shading, parameterName=Ignore Z use SkinFX mask, sortPriority=0}
  n217 = AppendVector(A=Constant2Vector() {g=1, r=1}, B=Saturate(None=OneMinus(None=@n216)))
  n218 = NamedRerouteUsage()
  n219 = DotProduct(A=Normalize(VectorInput=Multiply(A=@n215, B=@n217)), B=Normalize(VectorInput=Multiply(A=@n217, B=@n218))) {desc=Blinn}
  n220 = ComponentMask(None=NamedRerouteUsage()) {b=true, g=true, r=true}
  n221 = DotProduct(A=ReflectionVectorWS(), B=Normalize(VectorInput=Multiply(A=@n217, B=@n220))) {desc=Phong}
  n222 = ScalarParameter() {desc=Interpolate between Blinn and Phong Highlight, group=Toon_Shading_Fake PBR, parameterName=BlendBlinnPhong, sliderMax=1}
  n223 = SmoothStep(
   Min=OneMinus(None=LinearInterpolate(A=@n210, B=@n211, Alpha=@n212))
   Max=Constant() {r=1}
   Value=Saturate(None=LinearInterpolate(A=@n219, B=@n221, Alpha=@n222))
  )
  n224 = Divide(A=@n223) {constB=5}
  n225 = CurveAtlasRowParameter(CurveTime=SmoothStep(Value=Power(Base=@n224) {constExponent=0.5})) {
   atlas=&AthenaCharacterCurveAtlas
   curve=&TwoToneShadingCurve
   defaultValue=14
   group=Toon_Shading_Fake PBR
   parameterName=SpecularShadingCurve
   sortPriority=5
  }
  n226 = MaterialFunctionCall(In=@n223, Contrast=ScalarParameter() {defaultValue=2, group=Toon_Shading_Fake PBR, parameterName=Specular Sharpness, sliderMax=2}) {materialFunction=&CheapContrast}
  n227 = Round(None=@n223)
  n228 = ScalarParameter() {group=Toon_Shading_Fake PBR, parameterName=Bool Specular, sliderMax=1}
  n229 = Multiply(
   A=Multiply(A=LinearInterpolate(A=@n205.RGB, B=@n206, Alpha=@n207), B=Multiply(A=@n208, B=@n209) {constB=0.25})
   B=StaticSwitchParameter(
    True=ChannelMaskParameter(None=@n225) {group=Toon_Shading_Fake PBR, parameterName=SpecularShadingCurveChannel, sortPriority=10}.R
    False=LinearInterpolate(A=@n226.Result, B=@n227, Alpha=@n228)
   ) {group=Toon_Shading_Fake PBR, parameterName=UseSpecularCurve}
  )
  n230 = NamedRerouteUsage()
  n231 = NamedRerouteDeclaration(None=Multiply(A=Multiply(A=@n229, B=@n230) {constA=1}, B=NamedRerouteUsage())) {name=FakePBR_RefProfile, nodeColor=<rgba 1 0.8906 0 1>}
  n232 = NamedRerouteUsage()
  NamedRerouteDeclaration(None=Add(A=Multiply(A=@n231, B=@n232), B=NamedRerouteUsage())) {name=Fake PBR, nodeColor=<rgba 0 0.2266 1 1>}
  n233 = ChannelMaskParameter(None=@n194) {group=Toon_Shading_SSC, maskChannel=Blue, parameterName=SSC_SheenMaskChannel, sortPriority=10, defaultValue=<rgba 0 0 1 0>}
  NamedRerouteDeclaration(None=@n233.R) {name=SSC_Blue, nodeColor=<rgba 0 0.1118 1 1>}
  NamedRerouteDeclaration(None=@n128.B) {name=Roughness Input, nodeColor=<rgba 1 1 0 1>}
  PinBase() {reroutePins [LightVector {expression=&self:MaterialExpressionReroute_12}]}
  PinBase!(Result=@n140) {reroutePins [Result {expression=&self:MaterialExpressionReroute_14}]}
  SkyAtmosphereLightDirection()
 ]
}
