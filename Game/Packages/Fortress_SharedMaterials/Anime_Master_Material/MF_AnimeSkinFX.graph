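# MF_AnimeSkinFX is a material function that layers optional-looking skin effects onto the material attributes of the anime master material.
# Attributes pass in series through MFC_EmeraldGreenCowl, MFC_ArcticIceCape, MFC_MetalLUT_Toon, MFC_ThinFilm_Toon and MFC_ClothWind.
# Named reroutes carry shared inputs, including SkinFX, SSC and InkLine Mask.
# The Shading input splits by channel into Shading Curve, Specular Highlights and TOD Adjustment.
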MaterialFunction /Game/Packages/Fortress_SharedMaterials/Anime_Master_Material/MF_AnimeSkinFX
graph {nodes [
  n1 = FunctionInput() {inputName=MA, inputType=FunctionInput_MaterialAttributes, sortPriority=10}
  n2 = NamedRerouteUsage()
  n3 = NamedRerouteUsage()
  n4 = NamedRerouteUsage()
  n5 = MaterialFunctionCall!(
   MA=MaterialFunctionCall!(MA=MaterialFunctionCall!(MA=@n1, SC=@n2, FX=@n3) {materialFunction=&/MFC_EmeraldGreenCowl}.Result) {materialFunction=&/MFC_ArcticIceCape}.Result
   FX=@n4
  ) {materialFunction=&/MFC_MetalLUT_Toon}
  n6 = NamedRerouteUsage()
  FunctionOutput(None=MaterialFunctionCall!(
    Material=MaterialFunctionCall!(MA=@n5.Result, FX=@n4, SSC=@n6) {materialFunction=&/MFC_ThinFilm_Toon}.Result
    FX=NamedRerouteUsage()
   ) {materialFunction=&/MFC_ClothWind}.Result) {outputName=MA, sortPriority=10}
  n7 = FunctionInput() {bUsePreviewValueAsDefault=true, inputName=Shading, sortPriority=20}
  NamedRerouteDeclaration(None=ComponentMask(None=@n7) {r=true}) {name=Shading Curve, nodeColor=<rgba 1 0.375 0 1>}
  NamedRerouteDeclaration(None=FunctionInput() {bUsePreviewValueAsDefault=true, inputName=InkLine Mask, inputType=FunctionInput_Scalar, sortPriority=30}) {name=InkLine Mask, nodeColor=<rgba 0.6406 0 1 1>}
  NamedRerouteDeclaration(None=FunctionInput() {bUsePreviewValueAsDefault=true, inputName=SSC, inputType=FunctionInput_Vector4, sortPriority=40}) {name=SSC, nodeColor=<rgba 0.6406 0 1 1>}
  NamedRerouteDeclaration(None=FunctionInput() {bUsePreviewValueAsDefault=true, inputName=SkinFX, inputType=FunctionInput_Vector4, sortPriority=60}) {name=SkinFX, nodeColor=<rgba 1 0 0.5859 1>}
  NamedRerouteDeclaration(None=ComponentMask(None=@n7) {g=true}) {name=Specular Highlights, nodeColor=<rgba 1 0.7166 0 1>}
  NamedRerouteDeclaration(None=ComponentMask(None=@n7) {b=true}) {name=TOD Adjustment, nodeColor=<rgba 0.3412 0.3012 1 1>}
 ]}
