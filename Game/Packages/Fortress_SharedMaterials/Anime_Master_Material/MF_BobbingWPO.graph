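# Anime master material function that optionally adds bobbing and rotational sway to a mesh through World Position Offset.
# Sine-over-time drives an oscillating local Z-axis rotation and a vertical bob, each with speed and amount parameters, added to the incoming offset.
# Normals are corrected to follow the rotation.
# Static switches either bypass the effect or replace the offset with MF_Creative_CoinBobble output for Creative-mode coins.
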MaterialFunction /Game/Packages/Fortress_SharedMaterials/Anime_Master_Material/MF_BobbingWPO
graph {nodes [
  n1 = GetMaterialAttributes(None=FunctionInput() {inputType=FunctionInput_MaterialAttributes}) {attributeGetTypes [#1, #2]}
  n2 = SetMaterialAttributes!(
   MaterialAttributes=@n1.MaterialAttributes
   WorldPositionOffset=MaterialFunctionCall!(TimeScale, MaxHeight, MinHeight) {materialFunction=/Game/Playgrounds/Items/Props/Materials/MF_Creative_CoinBobble}.WorldPositionOffset
  ) {attributeSetTypes [#1]}
  n3 = Constant3Vector() {constant=<rgba 0 0 1 0>}
  n4 = MaterialFunctionCall(Speed=ScalarParameter() {defaultValue=0.5, group=BobbingWPO, parameterName=RotatingSpeed}) {materialFunction=&DebugTimeSine}
  n5 = Multiply(A=Subtract(A=Multiply(A=@n4.Result) {constB=2}), B=ScalarParameter() {group=BobbingWPO, parameterName=RotatingAmount})
  n6 = Transform(None=RotateAboutAxis(
    NormalizedRotationAxis=@n3
    RotationAngle=@n5
    PivotPoint=Constant()
    Position=MaterialFunctionCall() {materialFunction=&LocalPosition}.Local Position
   )) {transformSourceType=TRANSFORMSOURCE_Local}
  n7 = Constant()
  n8 = MaterialFunctionCall(Speed=ScalarParameter() {defaultValue=0.5, group=BobbingWPO, parameterName=BobbingSpeed}) {materialFunction=&DebugTimeSine}
  n9 = ScalarParameter() {defaultValue=10, group=BobbingWPO, parameterName=BobbingVerticalHeight}
  n10 = MaterialFunctionCall(
   X=@n7
   Y=@n7
   Z=LinearInterpolate(B=Multiply(A=@n8.Result, B=@n9) {constB=10}, Alpha=ScalarParameter() {group=BobbingWPO, parameterName=BobbingAmount})
  ) {materialFunction=&MakeFloat3}
  n11 = SetMaterialAttributes!(
   MaterialAttributes=@n1.MaterialAttributes
   WorldPositionOffset=Add(A=@n1.WorldPositionOffset, B=Add(A=@n6, B=@n10.Result))
   Normal=MaterialFunctionCall(Rotation Axis=@n3, Rotation Angle=@n5, World Space Vertex Normals=Transform(None=@n1.Normal)) {materialFunction=&FixRotateAboutAxisNormals}.New Tangent Space Vertex Normal
  ) {attributeSetTypes [#1, #2]}
  FunctionOutput(None=StaticSwitchParameter(
    True=StaticSwitchParameter(True=@n2, False=@n11) {group=BobbingWPO, parameterName=CreativeMode_CoinBob, sortPriority=0}
    False=@n1.MaterialAttributes
   ) {group=Skin FX, parameterName=Use BobbingWPO})
 ]}
