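# Material function in the anime master material set that adds pickup shrink and black hole vertex offsets to incoming material attributes.
# LODEffect takes the offset world position plus a distance-cull fade value from PickupEffectDistanceCullOverride.
# BlackHoleFunction takes that shifted position and the WorldPostionOfBlackHole parameter, and both results sum into World Position Offset.
# The "Use PickupShrink" static switch selects this result or passes attributes through unchanged.
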MaterialFunction /Game/Packages/Fortress_SharedMaterials/Anime_Master_Material/MF_PickupShrinking
graph {nodes [
  n1 = VectorParameter() {parameterName=WorldPostionOfBlackHole, defaultValue=<rgba -30000 25536 1952 0>}
  n2 = GetMaterialAttributes(None=FunctionInput() {inputType=FunctionInput_MaterialAttributes}) {attributeGetTypes [#1]}
  n3 = MaterialFunctionCall!(
   World Position=Add(A=@n2.WorldPositionOffset, B=WorldPosition().XYZ)
   DistanceCullFadeAnimationOffset=MaterialFunctionCall() {materialFunction=/Game/Packages/Fortress_SharedMaterials/Base_Material_Functions/PickupEffectDistanceCullOverride}.Result
   Use actor center instead of pivot point
  ) {materialFunction=/Game/Packages/Fortress_SharedMaterials/Base_Material_Functions/LODEffect}
  n4 = WorldPosition()
  n5 = Add(
   A=MaterialFunctionCall!(WorldPostionOfBlackHole=AppendVector(A=@n1.RGB, B=@n1.RGB), World Position=Add(A=@n3.Result, B=@n4.XYZ)) {materialFunction=/Game/Packages/Fortnite_UniqueMaterials_1/BlackHoleFunction}.Result
   B=@n3.Result
  )
  FunctionOutput(None=StaticSwitchParameter(
    True=SetMaterialAttributes!(MaterialAttributes=@n2.MaterialAttributes, WorldPositionOffset=@n5) {attributeSetTypes [#1]}
    False=@n2.MaterialAttributes
   ) {parameterName=Use PickupShrink})
 ]}
