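# Material function in the anime master material folder that re-expresses an input vector in a per-pixel, screen-aligned basis.
# It builds right from the camera's transformed up axis crossed with the normalized camera-relative position, then derives up from that ray and right.
# The normalized input's X, Y and Z weight right, up and the view ray, so the direction follows the screen with perspective correction.
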MaterialFunction /Game/Packages/Fortress_SharedMaterials/Anime_Master_Material/MF_ProjectToScreenSpace
graph {nodes [
  n1 = Normalize(VectorInput=FunctionInput() {inputName=Vector, previewValue=<0 0 1 1>})
  n2 = Constant3Vector() {constant=<rgba 0 1 0 0>}
  n3 = Normalize(VectorInput=WorldPosition() {worldPositionShaderOffset=WPT_CameraRelative}.XYZ)
  n4 = Normalize(VectorInput=CrossProduct(A=Transform(None=@n2) {transformSourceType=TRANSFORMSOURCE_View}, B=@n3))
  n5 = Multiply(A=ComponentMask(None=@n1) {r=true}, B=@n4)
  n6 = Multiply(A=ComponentMask(None=@n1) {g=true}, B=CrossProduct(A=@n3, B=@n4))
  n7 = ComponentMask(None=@n1) {b=true}
  FunctionOutput(None=Add(A=Add(A=@n5, B=@n6), B=Multiply(A=@n7, B=@n3)))
 ]}
