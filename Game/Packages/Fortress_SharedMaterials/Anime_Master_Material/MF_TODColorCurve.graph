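# Material function that scales an input color by a time-of-day brightness curve for the shared anime character master material.
# It samples a curve atlas row parameter, defaulting to the toon brightness curve in the Athena character atlas, at the TimeOfDay_0-1 value from the Fortnite parameter collection.
# A static channel mask parameter, red by default, picks the curve channel that multiplies the color, so instances can swap curve and channel.
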MaterialFunction /Game/Packages/Fortress_SharedMaterials/Anime_Master_Material/MF_TODColorCurve
graph {
 properties {bExposeToLibrary=true}
 nodes [
  n1 = CurveAtlasRowParameter(CurveTime=CollectionParameter() {collection=&FortniteMaterialParameters, parameterName=TimeOfDay_0-1}) {
   atlas=/Game/Characters/Shared/CurveAtlas/AthenaCharacterCurveAtlas
   curve=&/ToonTODBrightness_Curve
   defaultValue=17
   group=TOD_Adjustments
   parameterName=TODBrightnessAdjustmentCurve
  }
  FunctionOutput(None=Multiply(
    A=FunctionInput() {inputName=Color}
    B=StaticComponentMaskParameter(None=@n1) {defaultR=true, group=TOD_Adjustments, parameterName=TODCurveChannelMask}
   ))
 ]
}
