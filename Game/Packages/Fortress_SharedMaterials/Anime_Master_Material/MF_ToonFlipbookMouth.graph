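# Material function that overlays an animated toon flipbook mouth on a character material's emissive color, behind the UseToonFlipbookMouth static switch.
# Thresholded facial pose scalars select the frame: an open jaw picks open smile, frown or default frames, otherwise smile, frown, narrow or default apply in priority order.
# SubUV samples that frame on UV1, blends it over emissive using its RGB as alpha, then passes it through MF_TODColorCurve.
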MaterialFunction /Game/Packages/Fortress_SharedMaterials/Anime_Master_Material/MF_ToonFlipbookMouth
graph {
 properties {description=10}
 nodes [
  n1 = FunctionInput() {inputType=FunctionInput_MaterialAttributes}
  n2 = GetMaterialAttributes(None=@n1) {attributeGetTypes [#1]}
  n3 = TextureObjectParameter() {
   group=ToonFlipbookMouth
   parameterName=ToonMouthFlipbookTexture
   sortPriority=1
   texture=/Game/Accessories/FORT_Backpacks/M_LRG_Bob_Backpack/Textures/T_8BitGhoul
  }
  n4 = ScalarParameter() {defaultValue=0.1, group=ToonFlipbookMouth, parameterName=THreshold}
  n5 = ScalarParameter() {group=ToonFlipbookMouth, parameterName=Default Open Mouth}
  n6 = If(
   A=ScalarParameter() {group=ToonFlipbookMouth, parameterName=R_Frown_pose}
   B=@n4
   "A > B"=ScalarParameter() {defaultValue=10, group=ToonFlipbookMouth, parameterName=Mouth Open Frown Frame}
   "A == B"=@n5
   "A < B"=@n5
  )
  n7 = If(
   A=ScalarParameter() {group=ToonFlipbookMouth, parameterName=R_smile_pose}
   B=@n4
   "A > B"=ScalarParameter() {defaultValue=6, group=ToonFlipbookMouth, parameterName=Mouth Open Smile Frame}
   "A == B"=@n5
   "A < B"=@n5
  )
  n8 = ScalarParameter() {defaultValue=1, group=ToonFlipbookMouth, parameterName=Default Face}
  n9 = If(
   A=Max(
    A=ScalarParameter() {group=ToonFlipbookMouth, parameterName=R_lip_corner_narrow_pose}
    B=ScalarParameter() {group=ToonFlipbookMouth, parameterName=PeelyBlenderFace}
   )
   B=@n4
   "A > B"=ScalarParameter() {defaultValue=3, group=ToonFlipbookMouth, parameterName=Narrow Frame}
   "A == B"=@n8
   "A < B"=@n8
  )
  n10 = If(
   A=ScalarParameter() {group=ToonFlipbookMouth, parameterName=R_frown_pose}
   B=@n4
   "A > B"=ScalarParameter() {defaultValue=4, group=ToonFlipbookMouth, parameterName=Frown Frame}
   "A == B"=@n9
   "A < B"=@n9
  )
  n11 = If(
   A=ScalarParameter() {group=ToonFlipbookMouth, parameterName=R_smile_pose}
   B=@n4
   "A > B"=ScalarParameter() {defaultValue=8, group=ToonFlipbookMouth, parameterName=Smile Mouth animation Frame}
   "A == B"=@n10
   "A < B"=@n10
  )
  n12 = MaterialFunctionCall!(
   Texture=@n3
   UVs=TextureCoordinate() {coordinateIndex=1}
   SubImages=ScalarParameter() {defaultValue=4, group=ToonFlipbookMouth, parameterName=SubImages}
   Frame=If(A=ScalarParameter() {group=ToonFlipbookMouth, parameterName=Jaw_open_Pose}, B=@n4, "A > B"=Max(A=@n6, B=@n7), "A == B"=@n11, "A < B"=@n11)
  ) {materialFunction=/Engine/Functions/Engine_MaterialFunctions03/Texturing/SubUV_Function}
  n13 = MaterialFunctionCall!(Color=LinearInterpolate(A=@n2.EmissiveColor, B=@n12.RGB, Alpha=@n12.RGB)) {materialFunction=&/MF_TODColorCurve}
  FunctionOutput(None=StaticSwitchParameter(
    True=SetMaterialAttributes!(MaterialAttributes=@n2.MaterialAttributes, EmissiveColor=@n13.Result) {attributeSetTypes [#1]}
    False=@n1
   ) {group=Skin FX, parameterName=UseToonFlipbookMouth})
  Max(A=Max(A=@n9, B=@n10), B=@n11)
 ]
}
