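# Material function that adds a dark toon outline to a material on the mobile shading path.
# A Fresnel on the pixel normal is inverted, raised to a power, remapped by a lerp and saturated into a sharpened mask that is dark at grazing angles.
# The mask multiplies Base Color and Emissive, with tuning exposed as ToonOutline_Mobile parameters.
# A shading path switch passes the input attributes through unchanged on deferred and forward paths.
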MaterialFunction /Game/Packages/Fortress_SharedMaterials/Anime_Master_Material/MF_ToonOutline_Mobile
graph {nodes [
  n1 = GetMaterialAttributes(None=FunctionInput() {inputName=Material, inputType=FunctionInput_MaterialAttributes}) {attributeGetTypes [#1, #2, #3]}
  n2 = OneMinus(None=Fresnel(ExponentIn=ScalarParameter() {defaultValue=6, group=ToonOutline_Mobile, parameterName=BL_OutlineExponent1}, Normal=PixelNormalWS()) {baseReflectFraction=0})
  n3 = ScalarParameter() {defaultValue=12, group=ToonOutline_Mobile, parameterName=BL_OutlineExponent2}
  n4 = Saturate(None=LinearInterpolate(
    A=ScalarParameter() {defaultValue=-4, group=ToonOutline_Mobile, parameterName=BL_OutlineC1}
    B=ScalarParameter() {defaultValue=12, group=ToonOutline_Mobile, parameterName=BL_OutlineC2}
    Alpha=Power(Base=@n2, Exp=@n3)
   ))
  n5 = SetMaterialAttributes!(MaterialAttributes=@n1.MaterialAttributes, BaseColor=Multiply(A=@n1.BaseColor, B=@n4), EmissiveColor=Multiply(A=@n1.EmissiveColor, B=@n4)) {attributeSetTypes [#2, #3]}
  FunctionOutput(None=SetMaterialAttributes!(MaterialAttributes=ShadingPathSwitch(Default=@n1.MaterialAttributes, Deferred=@n1.MaterialAttributes, Forward=@n1.MaterialAttributes, Mobile=@n5)))
 ]}
