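# Material function that converts an input set of material attributes into a toon look through MF_AnimeShading.
# Saturated base color, scaled by a brightness blended between texture and skin values via SkinMask, feeds emissive and the lit diffuse override.
# Clamped metallic, inverted roughness and specular form a sheen, quantized into quarter steps for the SSC alpha override.
# The shaded result gets saturated emissive and a Subsurface shading model.
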MaterialFunction /Game/Packages/Fortress_SharedMaterials/Anime_Master_Material/MF_Toonify
graph {nodes [
  n1 = FunctionInput() {inputType=FunctionInput_MaterialAttributes}
  n2 = GetMaterialAttributes(None=@n1) {attributeGetTypes [#1, #2, #3]}
  n3 = Multiply(
   A=LinearInterpolate(
    A=ScalarParameter() {defaultValue=0.15, parameterName=Input Texture Brightness}
    B=ScalarParameter() {defaultValue=1, parameterName=Skin Texture Brightness}
    Alpha=FunctionInput() {bUsePreviewValueAsDefault=true, inputName=SkinMask, inputType=FunctionInput_Scalar}
   )
   B=Saturate(None=Max(A=@n2.BaseColor, B=@n2.BaseColor))
  )
  n4 = GetMaterialAttributes(None=@n1) {attributeGetTypes [#4, #5, #6]}
  n5 = Multiply(
   A=Clamp(None=@n4.Metallic) {minDefault=0.4}
   B=Multiply(A=OneMinus(None=@n4.Roughness), B=ScalarParameter() {defaultValue=1.25, parameterName=Sheen Intensity})
  )
  n6 = Round(None=Multiply(A=Multiply(A=@n5, B=@n4.Specular)) {constB=4})
  n7 = GetMaterialAttributes(None=MaterialFunctionCall!(
    Material=SetMaterialAttributes!(MaterialAttributes=@n1, Normal=@n2.Normal, EmissiveColor=@n3) {attributeSetTypes [#3, #2]}
    LitDiffuseOverride=@n3
    SSC_AlphaOverride=Divide(A=@n6) {constB=4}
    Tint Mask
    DF InkLine Visibility Mask
    DF Inkline Injection
    DF Inkline Color Override
    DF Inkline Color Override Mask
    CustomUVs
   ) {materialFunction=&/MF_AnimeShading}.Result) {attributeGetTypes [#2]}
  FunctionOutput(None=SetMaterialAttributes!(
    MaterialAttributes=@n7.MaterialAttributes
    EmissiveColor=Saturate(None=@n7.EmissiveColor)
    ShadingModel=ShadingModel() {shadingModel=MSM_Subsurface}
   ) {attributeSetTypes [#2, #7]})
 ]}
